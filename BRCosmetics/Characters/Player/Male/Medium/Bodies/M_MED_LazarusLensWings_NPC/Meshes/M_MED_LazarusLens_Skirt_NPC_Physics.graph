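# Physics asset for the LazarusLens NPC skirt, with capsule bodies on the dyn_dress bones, pelvis, thighs and calves.
# Three back chains hang from a free-swinging base on the pelvis and two front chains attach directly to it, each three bones long with mostly tight angular limits.
# Back tips lock Swing2, and angularly free constraints tie the middle tip to both side tips, cross-linking the back panel.
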PhysicsAsset /BRCosmetics/Characters/Player/Male/Medium/Bodies/M_MED_LazarusLensWings_NPC/Meshes/M_MED_LazarusLens_Skirt_NPC_Physics
bodies {
 calf_l [{shapeType=Capsule, length=29, radius=7.5, shapeName=calf_l_capsule, center=<-27.373 0 0>, rotation=<rot 90 0 0>}]
 calf_r [{shapeType=Capsule, length=28, radius=7.5, shapeName=calf_r_capsule, center=<28.0341 0 0>, rotation=<rot 90 0 0>}]
 dyn_dress_back_01_l [{shapeType=Capsule, length=13.5, radius=5, shapeName=dyn_dress_back_01_l_capsule, center=<5 0 0>}]
 dyn_dress_back_01_mid [{shapeType=Capsule, length=13.5, radius=5, shapeName=dyn_dress_back_01_mid_capsule, center=<5 0 0>}]
 dyn_dress_back_01_r [{shapeType=Capsule, length=13.5, radius=5, shapeName=dyn_dress_back_01_r_capsule, center=<5 0 0>}]
 dyn_dress_back_02_l [{shapeType=Capsule, length=13.5, radius=5, shapeName=dyn_dress_back_02_l_capsule, center=<5 0 0>}]
 dyn_dress_back_02_mid [{shapeType=Capsule, length=13.5, radius=5, shapeName=dyn_dress_back_02_mid_capsule, center=<5 0 0>}]
 dyn_dress_back_02_r [{shapeType=Capsule, length=13.5, radius=5, shapeName=dyn_dress_back_02_r_capsule, center=<5 0 0>}]
 dyn_dress_back_03_l [{shapeType=Capsule, length=11, radius=5, shapeName=dyn_dress_back_03_l_capsule, center=<5 0 0>}]
 dyn_dress_back_03_mid [{shapeType=Capsule, length=15.5, radius=5, shapeName=dyn_dress_back_03_mid_capsule, center=<5 0 0>}]
 dyn_dress_back_03_r [{shapeType=Capsule, length=11, radius=5, shapeName=dyn_dress_back_03_r_capsule, center=<5 0 0>}]
 dyn_dress_back_base [{shapeType=Capsule, length=13.5, radius=5, shapeName=dyn_dress_back_base_capsule}]
 dyn_dress_front_01_l [{shapeType=Capsule, length=13.5, radius=5, shapeName=dyn_dress_back_01_l_capsule}]
 dyn_dress_front_01_r [{shapeType=Capsule, length=13.5, radius=5, shapeName=dyn_dress_back_01_l_capsule}]
 dyn_dress_front_02_l [{shapeType=Capsule, length=13.5, radius=5, shapeName=dyn_dress_back_02_l_capsule}]
 dyn_dress_front_02_r [{shapeType=Capsule, length=13.5, radius=5, shapeName=dyn_dress_back_02_l_capsule}]
 dyn_dress_front_03_l [{shapeType=Capsule, length=12, radius=5, shapeName=dyn_dress_back_03_l_capsule}]
 dyn_dress_front_03_r [{shapeType=Capsule, length=12, radius=5, shapeName=dyn_dress_back_03_l_capsule}]
 pelvis [{shapeType=Capsule, length=7, radius=4.9729}]
 thigh_l [{shapeType=Capsule, length=21.5, radius=7, shapeName=thigh_l_capsule, center=<-27.373 0 0>, rotation=<rot 90 0 0>}]
 thigh_r [{shapeType=Capsule, length=20.5, radius=7, shapeName=thigh_r_capsule, center=<28.0341 0 0>, rotation=<rot 90 0 0>}]
}
constraints [
 {
  bone1Name=dyn_dress_back_base
  bone2Name=pelvis
  swing1LimitDegrees=45
  swing1Motion=Free
  swing2LimitDegrees=45
  swing2Motion=Free
  twistLimitDegrees=45
  twistMotion=Free
 }
 {bone1Name=dyn_dress_back_01_l, bone2Name=dyn_dress_back_base, swing1LimitDegrees=15, swing2LimitDegrees=15, twistLimitDegrees=5}
 {bone1Name=dyn_dress_back_02_l, bone2Name=dyn_dress_back_01_l, swing1LimitDegrees=15, swing2LimitDegrees=5, twistLimitDegrees=5}
 {
  bone1Name=dyn_dress_back_03_l
  bone2Name=dyn_dress_back_02_l
  swing1LimitDegrees=15
  swing2LimitDegrees=5
  swing2Motion=Locked
  twistLimitDegrees=5
 }
 {bone1Name=dyn_dress_back_01_mid, bone2Name=dyn_dress_back_base, swing1LimitDegrees=15, swing2LimitDegrees=15, twistLimitDegrees=5}
 {bone1Name=dyn_dress_back_02_mid, bone2Name=dyn_dress_back_01_mid, swing1LimitDegrees=15, swing2LimitDegrees=5, twistLimitDegrees=5}
 {
  bone1Name=dyn_dress_back_03_mid
  bone2Name=dyn_dress_back_02_mid
  swing1LimitDegrees=15
  swing2LimitDegrees=5
  swing2Motion=Locked
  twistLimitDegrees=5
 }
 {bone1Name=dyn_dress_back_01_r, bone2Name=dyn_dress_back_base, swing1LimitDegrees=15, swing2LimitDegrees=15, twistLimitDegrees=5}
 {bone1Name=dyn_dress_back_02_r, bone2Name=dyn_dress_back_01_r, swing1LimitDegrees=15, swing2LimitDegrees=5, twistLimitDegrees=5}
 {
  bone1Name=dyn_dress_back_03_r
  bone2Name=dyn_dress_back_02_r
  swing1LimitDegrees=15
  swing2LimitDegrees=5
  swing2Motion=Locked
  twistLimitDegrees=5
 }
 {
  bone1Name=thigh_l
  bone2Name=pelvis
  swing1LimitDegrees=45
  swing1Motion=Free
  swing2LimitDegrees=45
  swing2Motion=Free
  twistLimitDegrees=45
  twistMotion=Free
 }
 {
  bone1Name=calf_l
  bone2Name=thigh_l
  swing1LimitDegrees=45
  swing1Motion=Free
  swing2LimitDegrees=45
  swing2Motion=Free
  twistLimitDegrees=45
  twistMotion=Free
 }
 {
  bone1Name=thigh_r
  bone2Name=pelvis
  swing1LimitDegrees=45
  swing1Motion=Free
  swing2LimitDegrees=45
  swing2Motion=Free
  twistLimitDegrees=45
  twistMotion=Free
 }
 {
  bone1Name=calf_r
  bone2Name=thigh_r
  swing1LimitDegrees=45
  swing1Motion=Free
  swing2LimitDegrees=45
  swing2Motion=Free
  twistLimitDegrees=45
  twistMotion=Free
 }
 {bone1Name=dyn_dress_front_01_l, bone2Name=pelvis, swing1LimitDegrees=15, swing2LimitDegrees=15, twistLimitDegrees=5}
 {bone1Name=dyn_dress_front_02_l, bone2Name=dyn_dress_front_01_l, swing1LimitDegrees=45, swing2LimitDegrees=8, twistLimitDegrees=5}
 {bone1Name=dyn_dress_front_03_l, bone2Name=dyn_dress_front_02_l, swing1LimitDegrees=30, swing2LimitDegrees=8, twistLimitDegrees=5}
 {bone1Name=dyn_dress_front_01_r, bone2Name=pelvis, swing1LimitDegrees=15, swing2LimitDegrees=15, twistLimitDegrees=5}
 {bone1Name=dyn_dress_front_02_r, bone2Name=dyn_dress_front_01_r, swing1LimitDegrees=30, swing2LimitDegrees=8, twistLimitDegrees=5}
 {bone1Name=dyn_dress_front_03_r, bone2Name=dyn_dress_front_02_r, swing1LimitDegrees=30, swing2LimitDegrees=8, twistLimitDegrees=5}
 {
  bone1Name=dyn_dress_back_03_mid
  bone2Name=dyn_dress_back_03_l
  swing1LimitDegrees=45
  swing1Motion=Free
  swing2LimitDegrees=45
  swing2Motion=Free
  twistLimitDegrees=45
  twistMotion=Free
 }
 {
  bone1Name=dyn_dress_back_03_mid
  bone2Name=dyn_dress_back_03_r
  swing1LimitDegrees=45
  swing1Motion=Free
  swing2LimitDegrees=45
  swing2Motion=Free
  twistLimitDegrees=45
  twistMotion=Free
 }
]
