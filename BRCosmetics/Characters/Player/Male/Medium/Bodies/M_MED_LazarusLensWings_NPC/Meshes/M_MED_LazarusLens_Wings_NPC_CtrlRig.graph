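# Control Rig for the LazarusLens Wings NPC character body that makes dress bones follow the legs.
# From Begin Execution, two chained parent constraints bind each front dress master bone fully to its same-side thigh.
# A second pair blends the back dress bases between pelvis and both thighs, favoring the same side, but is not wired to Begin Execution.
# The hierarchy also holds wing and tassel bones, which the graph leaves untouched.
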ControlRigBlueprint /BRCosmetics/Characters/Player/Male/Medium/Bodies/M_MED_LazarusLensWings_NPC/Meshes/M_MED_LazarusLens_Wings_NPC_CtrlRig
graphs [
 RigVMFunctionLibrary {links [], nodes []}
 RigVMModel {nodes [
   &RigUnit_ParentConstraint(
    AdvancedSettings=(InterpolationType=Average,RotationOrderForFilter=XZY)
    Child=(Type=Bone,Name="dyn_dress_back_base_l")
    ExecutePin=()
    Filter=(TranslationFilter=(bX=True,bY=True,bZ=True),RotationFilter=(bX=true,bY=True,bZ=true),ScaleFilter=(bX=True,bY=True,bZ=True))
    Parents=((Item=(Type=Bone,Name="pelvis"),Weight=0.25),(Item=(Type=Bone,Name="thigh_l"),Weight=0.7),(Item=(Type=Bone,Name="thigh_r"),Weight=0.05))
    Weight=1
   )
   > &RigUnit_ParentConstraint(
    AdvancedSettings=(InterpolationType=Average,RotationOrderForFilter=XZY)
    Child=(Type=Bone,Name="dyn_dress_back_base_r")
    Filter=(TranslationFilter=(bX=True,bY=True,bZ=True),RotationFilter=(bX=true,bY=True,bZ=true),ScaleFilter=(bX=True,bY=True,bZ=True))
    Parents=((Item=(Type=Bone,Name="pelvis"),Weight=0.25),(Item=(Type=Bone,Name="thigh_r"),Weight=0.7),(Item=(Type=Bone,Name="thigh_l"),Weight=0.05))
    Weight=1
   )
   RigUnit_BeginExecution = &RigUnit_BeginExecution()
   &RigUnit_ParentConstraint(
    AdvancedSettings=(InterpolationType=Average,RotationOrderForFilter=XZY)
    Child=(Type=Bone,Name="dyn_dress_front_master_l")
    Filter=(TranslationFilter=(bX=True,bY=True,bZ=True),RotationFilter=(bX=true,bY=True,bZ=true),ScaleFilter=(bX=True,bY=True,bZ=True))
    Parents=((Item=(Type=Bone,Name="thigh_l"),Weight=1))
    Weight=1
    ExecutePin=@RigUnit_BeginExecution.ExecutePin
   )
   > &RigUnit_ParentConstraint(
    AdvancedSettings=(InterpolationType=Average,RotationOrderForFilter=XZY)
    Child=(Type=Bone,Name="dyn_dress_front_master_r")
    Filter=(TranslationFilter=(bX=True,bY=True,bZ=True),RotationFilter=(bX=true,bY=True,bZ=true),ScaleFilter=(bX=True,bY=True,bZ=True))
    Parents=((Item=(Type=Bone,Name="thigh_r"),Weight=1))
    Weight=1
   )
  ]}
]
hierarchy [
 root {loc=<0 0 0>, rot=<0 0 0 1>}
 pelvis {parent=root, loc=<0 -0.7071 93.1191>, rot=<0.0127 0.707 -0.0127 -0.707>}
 spine_01 {parent=pelvis, loc=<6.7494 -0.6388 0>, rot=<0 0 0.034 -0.9994>}
 spine_02 {parent=spine_01, loc=<7.2518 -0.035 0>, rot=<0 0 0.0238 -0.9997>}
 spine_03 {parent=spine_02, loc=<6.0104 0 0>, rot=<0 0 0.0149 -0.9999>}
 spine_04 {parent=spine_03, loc=<8.1024 0 0>, rot=<0 0 -0.0132 -0.9999>}
 spine_05 {parent=spine_04, loc=<10.0146 0 0>, rot=<0 0 -0.0015 -1>}
 attach_cape {parent=spine_05, loc=<9.017 -11.714 0>, rot=<0.0616 -0.7044 -0.0616 -0.7044>}
 neck_01 {parent=spine_05, loc=<11.3138 0.1529 0>, rot=<0 0 -0.2003 -0.9797>}
 neck_02 {parent=neck_01, loc=<4.5627 0 0>, rot=<0 0 0.0504 -0.9987>}
 head {parent=neck_02, loc=<5.3914 0 0>, rot=<0 0 0.115 -0.9934>}
 attach_backpack {parent=spine_05, loc=<-0.123 -13.473 0>, rot=<-0.0093 -0.707 0.0093 -0.707>}
 wings_base {parent=attach_backpack, loc=<0 0 0>, rot=<0 0 0 1>}
 wing_base_r {parent=wings_base, loc=<-10.3452 1.4096 8.2485>, rot=<0.0376 0.0376 0.7061 -0.7061>}
 wing_shoulder_r {parent=wing_base_r, loc=<1.3064 0.8081 -0.0161>, rot=<0.5925 -0.386 0.5374 -0.4596>}
 wing_elbow_r {parent=wing_shoulder_r, loc=<33.1889 0 0>, rot=<0.0051 -0.0066 0.366 -0.9306>}
 wing_hand_r {parent=wing_elbow_r, loc=<32.0385 0 0>, rot=<0 0 0 1>}
 wing_spread_b_base_r {parent=wing_hand_r, loc=<-4.2832 1.9895 -0.1502>, rot=<0.0526 0.0294 0.9213 0.3842>}
 dyn_wing_spread_b_01_r {parent=wing_spread_b_base_r, loc=<0 0 0>, rot=<0.1836 0 0 -0.983>}
 dyn_wing_spread_b_02_r {parent=dyn_wing_spread_b_01_r, loc=<11.961 0 0>, rot=<-0.0646 -0.0338 -0.1747 -0.9819>}
 dyn_wing_spread_b_03_r {parent=dyn_wing_spread_b_02_r, loc=<16.1692 0 0>, rot=<0.0101 0.0102 -0.11 -0.9938>}
 dyn_wing_spread_b_04_r {parent=dyn_wing_spread_b_03_r, loc=<18.014 0 0>, rot=<0.0113 0.0254 -0.0979 -0.9948>}
 dyn_wing_spread_b_05_r {parent=dyn_wing_spread_b_04_r, loc=<13.7065 0 0>, rot=<0.0095 0.0618 -0.1086 -0.9921>}
 dyn_wing_spread_b_06_r {parent=dyn_wing_spread_b_05_r, loc=<10.1814 0 0>, rot=<-0.0872 0 0 -0.9962>}
 wing_spread_c_base_r {parent=wing_hand_r, loc=<-2.6259 1.7548 -0.1325>, rot=<-0.0015 0.0189 -0.8164 -0.5772>}
 dyn_wing_spread_c_01_r {parent=wing_spread_c_base_r, loc=<0 0 0>, rot=<0 0 0 1>}
 dyn_wing_spread_c_02_r {parent=dyn_wing_spread_c_01_r, loc=<15.3158 0 0>, rot=<-0.0309 -0.0199 -0.1221 -0.9918>}
 dyn_wing_spread_c_03_r {parent=dyn_wing_spread_c_02_r, loc=<18.5453 0 0>, rot=<-0.0073 -0.0375 -0.1847 -0.982>}
 dyn_wing_spread_c_04_r {parent=dyn_wing_spread_c_03_r, loc=<15.6785 0 0>, rot=<0.0322 -0.0455 -0.2584 -0.9644>}
 dyn_wing_spread_c_05_r {parent=dyn_wing_spread_c_04_r, loc=<17.7935 0 0>, rot=<0.0704 0 0 -0.9975>}
 wing_spread_d_base_r {parent=wing_hand_r, loc=<-0.761 1.9943 -0.1506>, rot=<-0.0152 0.0143 -0.6279 -0.778>}
 dyn_wing_spread_d_01_r {parent=wing_spread_d_base_r, loc=<0 0 0>, rot=<0 0 0 1>}
 dyn_wing_spread_d_02_r {parent=dyn_wing_spread_d_01_r, loc=<16.7885 0 0>, rot=<-0.0013 -0.0444 -0.0526 -0.9976>}
 dyn_wing_spread_d_03_r {parent=dyn_wing_spread_d_02_r, loc=<17.592 0 0>, rot=<-0.0299 -0.0401 -0.2115 -0.9761>}
 dyn_wing_spread_d_04_r {parent=dyn_wing_spread_d_03_r, loc=<19.7829 0 0>, rot=<0.0077 -0.0402 -0.1884 -0.9812>}
 dyn_wing_spread_d_05_r {parent=dyn_wing_spread_d_04_r, loc=<15.3352 0 0>, rot=<0.0122 -0.0483 -0.2442 -0.9684>}
 dyn_wing_spread_d_06_r {parent=dyn_wing_spread_d_05_r, loc=<16.6876 0 0>, rot=<0 0 0 1>}
 wing_spread_e_base_r {parent=wing_hand_r, loc=<2.0242 1.3903 -0.105>, rot=<0.0104 -0.0173 -0.4754 -0.8796>}
 dyn_wing_spread_e_01_r {parent=wing_spread_e_base_r, loc=<0 0 0>, rot=<0 0 0 1>}
 dyn_wing_spread_e_02_r {parent=dyn_wing_spread_e_01_r, loc=<20.6859 0 0>, rot=<0.0001 -0.0075 -0.0373 -0.9993>}
 dyn_wing_spread_e_03_r {parent=dyn_wing_spread_e_02_r, loc=<19.8398 0 0>, rot=<-0.0012 -0.0125 -0.0616 -0.998>}
 dyn_wing_spread_e_04_r {parent=dyn_wing_spread_e_03_r, loc=<23.2751 0 0>, rot=<-0.0023 -0.011 -0.0531 -0.9985>}
 dyn_wing_spread_e_05_r {parent=dyn_wing_spread_e_04_r, loc=<18.5547 0 0>, rot=<-0.0237 -0.0474 -0.2018 -0.978>}
 dyn_wing_spread_e_06_r {parent=dyn_wing_spread_e_05_r, loc=<18.081 0 0>, rot=<0 0 0 1>}
 wing_spread_f_base_r {parent=wing_spread_e_base_r, loc=<0.3669 -1.9978 0.0706>, rot=<-0.0006 0.0046 0.0908 -0.9959>}
 dyn_wing_spread_f_01_r {parent=wing_spread_f_base_r, loc=<0 0 0>, rot=<0 0 0 1>}
 dyn_wing_spread_f_02_r {parent=dyn_wing_spread_f_01_r, loc=<24.3818 0 0>, rot=<0.0006 -0.0054 -0.121 -0.9926>}
 dyn_wing_spread_f_03_r {parent=dyn_wing_spread_f_02_r, loc=<20.6859 0 0>, rot=<-0.0007 -0.0051 -0.1141 -0.9935>}
 dyn_wing_spread_f_04_r {parent=dyn_wing_spread_f_03_r, loc=<22.0506 0 0>, rot=<-0.0017 -0.0047 -0.0991 -0.9951>}
 dyn_wing_spread_f_05_r {parent=dyn_wing_spread_f_04_r, loc=<15.6528 0 0>, rot=<0 0 0 1>}
 wing_hand_tip_01_r {parent=wing_hand_r, loc=<-0.5022 -0.5448 -0.033>, rot=<0.4014 -0.9156 -0.0133 -0.0185>}
 wing_hand_tip_02_r {parent=wing_hand_tip_01_r, loc=<3.6248 0 0>, rot=<0.0118 0.0294 -0.1718 -0.9846>}
 wing_hand_tip_03_r {parent=wing_hand_tip_02_r, loc=<8.5479 0 0>, rot=<0 0 0 1>}
 wing_spread_a_base_r {parent=wing_elbow_r, loc=<0 0 0>, rot=<-0.0482 -0.0882 0.9943 0.0349>}
 dyn_wing_spread_a_01_r {parent=wing_spread_a_base_r, loc=<0 0 0>, rot=<0 0 0 1>}
 dyn_wing_spread_a_02_r {parent=dyn_wing_spread_a_01_r, loc=<15.1614 0 0>, rot=<-0.0377 0.0237 -0.0962 -0.9944>}
 dyn_wing_spread_a_03_r {parent=dyn_wing_spread_a_02_r, loc=<19.8552 0 0>, rot=<-0.021 0.0203 -0.1095 -0.9936>}
 dyn_wing_spread_a_04_r {parent=dyn_wing_spread_a_03_r, loc=<9.0923 0 0>, rot=<-0.7619 0 0 -0.6477>}
 wing_shoulder_tip_01_r {parent=wing_base_r, loc=<1.3297 -1.0615 1.9597>, rot=<0.6496 -0.2794 0.6365 -0.308>}
 wing_shoulder_tip_02_r {parent=wing_shoulder_tip_01_r, loc=<4.7852 0 0>, rot=<0 0 0 1>}
 wing_base_l {parent=wings_base, loc=<10.3452 1.4096 8.2482>, rot=<-0.7061 -0.7061 0.0376 -0.0376>}
 wing_shoulder_l {parent=wing_base_l, loc=<-1.3064 -0.8081 0.016>, rot=<0.5925 -0.386 0.5374 -0.4596>}
 wing_elbow_l {parent=wing_shoulder_l, loc=<-33.1891 0.0005 0>, rot=<0.0051 -0.0066 0.366 -0.9306>}
 wing_hand_l {parent=wing_elbow_l, loc=<-32.0383 -0.0001 0>, rot=<0 0 0 1>}
 wing_spread_b_base_l {parent=wing_hand_l, loc=<4.2834 -1.9896 0.1503>, rot=<0.0526 0.0294 0.9213 0.3842>}
 dyn_wing_spread_b_01_l {parent=wing_spread_b_base_l, loc=<0 0 0>, rot=<0.1836 0 0 -0.983>}
 dyn_wing_spread_b_02_l {parent=dyn_wing_spread_b_01_l, loc=<-11.9605 0.0002 0>, rot=<-0.0646 -0.0338 -0.1747 -0.9819>}
 dyn_wing_spread_b_03_l {parent=dyn_wing_spread_b_02_l, loc=<-16.1691 0 0>, rot=<0.0101 0.0102 -0.11 -0.9938>}
 dyn_wing_spread_b_04_l {parent=dyn_wing_spread_b_03_l, loc=<-18.0142 0 0>, rot=<0.0113 0.0254 -0.0979 -0.9948>}
 dyn_wing_spread_b_05_l {parent=dyn_wing_spread_b_04_l, loc=<-13.7064 0 -0.0001>, rot=<0.0095 0.0618 -0.1086 -0.9921>}
 dyn_wing_spread_b_06_l {parent=dyn_wing_spread_b_05_l, loc=<-10.1814 0 0>, rot=<-0.0872 0 0 -0.9962>}
 wing_spread_c_base_l {parent=wing_hand_l, loc=<2.6261 -1.755 0.1326>, rot=<-0.0015 0.0189 -0.8164 -0.5772>}
 dyn_wing_spread_c_01_l {parent=wing_spread_c_base_l, loc=<0 0 0>, rot=<0 0 0 1>}
 dyn_wing_spread_c_02_l {parent=dyn_wing_spread_c_01_l, loc=<-15.3153 0.0003 0>, rot=<-0.0309 -0.0199 -0.1221 -0.9918>}
 dyn_wing_spread_c_03_l {parent=dyn_wing_spread_c_02_l, loc=<-18.5461 -0.0004 0>, rot=<-0.0073 -0.0375 -0.1847 -0.982>}
 dyn_wing_spread_c_04_l {parent=dyn_wing_spread_c_03_l, loc=<-15.6784 0 0>, rot=<0.0322 -0.0455 -0.2584 -0.9644>}
 dyn_wing_spread_c_05_l {parent=dyn_wing_spread_c_04_l, loc=<-17.7932 -0.0001 0>, rot=<0.0704 0 0 -0.9975>}
 wing_spread_d_base_l {parent=wing_hand_l, loc=<0.7608 -1.9943 0.1506>, rot=<-0.0152 0.0143 -0.6279 -0.778>}
 dyn_wing_spread_d_01_l {parent=wing_spread_d_base_l, loc=<0 0 0>, rot=<0 0 0 -1>}
 dyn_wing_spread_d_02_l {parent=dyn_wing_spread_d_01_l, loc=<-16.7885 -0.0001 0.0001>, rot=<-0.0013 -0.0444 -0.0526 -0.9976>}
 dyn_wing_spread_d_03_l {parent=dyn_wing_spread_d_02_l, loc=<-17.5923 -0.0005 0>, rot=<-0.0299 -0.0401 -0.2115 -0.9761>}
 dyn_wing_spread_d_04_l {parent=dyn_wing_spread_d_03_l, loc=<-19.7829 -0.0001 0>, rot=<0.0077 -0.0402 -0.1884 -0.9812>}
 dyn_wing_spread_d_05_l {parent=dyn_wing_spread_d_04_l, loc=<-15.3348 0.0001 0>, rot=<0.0122 -0.0483 -0.2442 -0.9684>}
 dyn_wing_spread_d_06_l {parent=dyn_wing_spread_d_05_l, loc=<-16.6881 -0.0003 -0.0001>, rot=<0 0 0 1>}
 wing_spread_e_base_l {parent=wing_hand_l, loc=<-2.0239 -1.3905 0.105>, rot=<0.0104 -0.0173 -0.4754 -0.8796>}
 dyn_wing_spread_e_01_l {parent=wing_spread_e_base_l, loc=<0 0 0>, rot=<0 0 0 -1>}
 dyn_wing_spread_e_02_l {parent=dyn_wing_spread_e_01_l, loc=<-20.686 0.0008 0>, rot=<0.0001 -0.0075 -0.0373 -0.9993>}
 dyn_wing_spread_e_03_l {parent=dyn_wing_spread_e_02_l, loc=<-19.8398 -0.0004 0>, rot=<-0.0012 -0.0125 -0.0616 -0.998>}
 dyn_wing_spread_e_04_l {parent=dyn_wing_spread_e_03_l, loc=<-23.2748 -0.0005 -0.0001>, rot=<-0.0023 -0.011 -0.0531 -0.9985>}
 dyn_wing_spread_e_05_l {parent=dyn_wing_spread_e_04_l, loc=<-18.5547 0.0006 0>, rot=<-0.0237 -0.0474 -0.2018 -0.978>}
 dyn_wing_spread_e_06_l {parent=dyn_wing_spread_e_05_l, loc=<-18.0811 0.0004 0>, rot=<0 0 0 -1>}
 wing_spread_f_base_l {parent=wing_spread_e_base_l, loc=<-0.3668 1.9977 -0.0705>, rot=<-0.0006 0.0046 0.0908 -0.9959>}
 dyn_wing_spread_f_01_l {parent=wing_spread_f_base_l, loc=<0 0 0>, rot=<0 0 0 1>}
 dyn_wing_spread_f_02_l {parent=dyn_wing_spread_f_01_l, loc=<-24.3819 0.0006 0>, rot=<0.0006 -0.0054 -0.121 -0.9926>}
 dyn_wing_spread_f_03_l {parent=dyn_wing_spread_f_02_l, loc=<-20.6863 -0.0001 -0.0001>, rot=<-0.0007 -0.0051 -0.1141 -0.9935>}
 dyn_wing_spread_f_04_l {parent=dyn_wing_spread_f_03_l, loc=<-22.0505 0.0003 0>, rot=<-0.0017 -0.0047 -0.0991 -0.9951>}
 dyn_wing_spread_f_05_l {parent=dyn_wing_spread_f_04_l, loc=<-15.6529 -0.0001 0.0001>, rot=<0 0 0 -1>}
 wing_hand_tip_01_l {parent=wing_hand_l, loc=<0.5023 0.5447 0.0331>, rot=<0.4014 -0.9156 -0.0133 -0.0185>}
 wing_hand_tip_02_l {parent=wing_hand_tip_01_l, loc=<-3.6248 0.0001 0>, rot=<0.0118 0.0294 -0.1718 -0.9846>}
 wing_hand_tip_03_l {parent=wing_hand_tip_02_l, loc=<-8.5478 -0.0001 0>, rot=<0 0 0 1>}
 wing_spread_a_base_l {parent=wing_elbow_l, loc=<0 0 0>, rot=<-0.0482 -0.0882 0.9943 0.0349>}
 dyn_wing_spread_a_01_l {parent=wing_spread_a_base_l, loc=<0 0 0>, rot=<0 0 0 1>}
 dyn_wing_spread_a_02_l {parent=dyn_wing_spread_a_01_l, loc=<-15.1615 0 0>, rot=<-0.0377 0.0237 -0.0962 -0.9944>}
 dyn_wing_spread_a_03_l {parent=dyn_wing_spread_a_02_l, loc=<-19.8551 -0.0001 0>, rot=<-0.021 0.0203 -0.1095 -0.9936>}
 dyn_wing_spread_a_04_l {parent=dyn_wing_spread_a_03_l, loc=<-9.0925 0.0002 0>, rot=<-0.7619 0 0 -0.6477>}
 wing_shoulder_tip_01_l {parent=wing_base_l, loc=<-1.3297 1.0615 -1.96>, rot=<0.6496 -0.2794 0.6365 -0.308>}
 wing_shoulder_tip_02_l {parent=wing_shoulder_tip_01_l, loc=<-4.785 -0.0003 0>, rot=<0 0 0 1>}
 clavicle_l {parent=spine_05, loc=<4.3575 -0.1466 -2.3223>, rot=<0.0465 -0.7289 -0.009 -0.6829>}
 upperarm_l {parent=clavicle_l, loc=<15.5111 0 0>, rot=<0.0226 -0.3234 -0.0088 -0.9459>}
 lowerarm_l {parent=upperarm_l, loc=<26.6514 0 0>, rot=<0.0391 -0.0099 -0.2453 -0.9686>}
 hand_l {parent=lowerarm_l, loc=<22.3491 0 0>, rot=<-0.7081 -0.0507 0.0975 -0.6975>}
 thumb_01_l {parent=hand_l, loc=<1.3118 -0.3191 -2.459>, rot=<0.6546 -0.4042 0.0664 -0.6354>}
 thumb_02_l {parent=thumb_01_l, loc=<5.9336 -0.0448 0.0401>, rot=<-0.0037 0.0213 0.298 -0.9543>}
 thumb_03_l {parent=thumb_02_l, loc=<2.2223 0.3682 -0.0907>, rot=<0.0118 -0.013 0.0066 -0.9998>}
 index_metacarpal_l {parent=hand_l, loc=<1.6535 0.0183 -2.96>, rot=<0.0158 -0.0124 -0.0173 -0.9996>}
 index_01_l {parent=index_metacarpal_l, loc=<8.3008 0.0357 -0.0524>, rot=<-0.0044 0.0485 0.0924 -0.9945>}
 index_02_l {parent=index_01_l, loc=<3.3587 -0.3004 0.0452>, rot=<-0.002 0.0027 0.3225 -0.9466>}
 index_03_l {parent=index_02_l, loc=<2.5578 -0.123 -0.0252>, rot=<-0.002 0.0067 0.2558 -0.9667>}
 middle_metacarpal_l {parent=hand_l, loc=<2.1856 -0.1725 -0.5984>, rot=<0.006 0.0168 -0.0329 -0.9993>}
 middle_01_l {parent=middle_metacarpal_l, loc=<7.3275 -0.2447 0.0129>, rot=<-0.0208 0.0042 0.0858 -0.9961>}
 middle_02_l {parent=middle_01_l, loc=<4.0123 0.1401 0.1175>, rot=<-0.0218 0.0258 0.3512 -0.9357>}
 middle_03_l {parent=middle_02_l, loc=<2.6077 0.0534 0.1373>, rot=<0.0082 0.0188 0.1708 -0.9851>}
 ring_metacarpal_l {parent=hand_l, loc=<2.2074 -0.3625 1.3638>, rot=<-0.0284 0.0722 -0.0458 -0.9959>}
 ring_01_l {parent=ring_metacarpal_l, loc=<7.0295 -0.6197 -0.1446>, rot=<-0.0489 -0.0252 0.1892 -0.9804>}
 ring_02_l {parent=ring_01_l, loc=<3.6174 0.4647 -0.1027>, rot=<0 0 0.2806 -0.9598>}
 ring_03_l {parent=ring_02_l, loc=<2.4152 0.114 0.0199>, rot=<0 0 0.1189 -0.9929>}
 pinky_metacarpal_l {parent=hand_l, loc=<1.7616 -0.9753 3.3814>, rot=<-0.0397 0.0814 -0.069 -0.9935>}
 pinky_01_l {parent=pinky_metacarpal_l, loc=<7.1709 -0.4831 -0.1094>, rot=<-0.0654 -0.0035 0.2242 -0.9723>}
 pinky_02_l {parent=pinky_01_l, loc=<2.8274 0.0893 -0.1043>, rot=<0.0037 0.0011 0.2854 -0.9584>}
 pinky_03_l {parent=pinky_02_l, loc=<2.2046 0.0307 -0.0078>, rot=<-0.0002 0.0098 0.1445 -0.9895>}
 hand_attach_l {parent=hand_l, loc=<8.5 -3 0>, rot=<0.7071 0.7071 0 0>}
 weapon_l {parent=hand_l, loc=<-0.1974 -3.4111 0.3811>, rot=<0 0 0 1>}
 lowerarm_twist_02_l {parent=lowerarm_l, loc=<9.6199 0 0>, rot=<0 0 0 -1>}
 lowerarm_twist_01_l {parent=lowerarm_l, loc=<17.8549 0 0>, rot=<0 0 0 -1>}
 deform_elbow_in_base_l {parent=lowerarm_l, loc=<0.2112 5.259 0.8274>, rot=<0 0 0 -1>}
 deform_elbow_in_l {parent=deform_elbow_in_base_l, loc=<0 0 0>, rot=<0 0 0 1>}
 deform_elbow_out_base_l {parent=lowerarm_l, loc=<0 -4 0>, rot=<0 0 0 -1>}
 deform_elbow_out_l {parent=deform_elbow_out_base_l, loc=<0 0 0>, rot=<0 0 0 1>}
 upperarm_twist_01_l {parent=upperarm_l, loc=<0.0041 0 0>, rot=<0 0 0 1>}
 deform_upperarm_bk_base_l {parent=upperarm_twist_01_l, loc=<1.0813 -7.2193 -1.9453>, rot=<0 0 0 1>}
 deform_upperarm_bk_l {parent=deform_upperarm_bk_base_l, loc=<0 0 0>, rot=<0 0 0 1>}
 deform_upperarm_fr_base_l {parent=upperarm_twist_01_l, loc=<-0.3691 7.0506 -2.765>, rot=<0 0 0 1>}
 deform_upperarm_fr_l {parent=deform_upperarm_fr_base_l, loc=<0 0 0>, rot=<0 0 0 1>}
 deform_upperarm_up_base_l {parent=upperarm_twist_01_l, loc=<-3.2216 -1.1391 3.3311>, rot=<0 0 0 1>}
 deform_upperarm_up_l {parent=deform_upperarm_up_base_l, loc=<0 0 0>, rot=<0 0 0 1>}
 upperarm_twist_02_l {parent=upperarm_l, loc=<13.7941 0 0>, rot=<0 0 0 1>}
 clavicle_r {parent=spine_05, loc=<4.3575 -0.1466 2.3223>, rot=<0.7289 0.0465 -0.6829 0.009>}
 upperarm_r {parent=clavicle_r, loc=<-15.5111 0 0>, rot=<0.0226 -0.3234 -0.0088 -0.9459>}
 lowerarm_r {parent=upperarm_r, loc=<-26.6514 0 0>, rot=<0.0391 -0.0099 -0.2453 -0.9686>}
 hand_r {parent=lowerarm_r, loc=<-22.3491 0 0>, rot=<-0.7081 -0.0507 0.0975 -0.6975>}
 weapon_r {parent=hand_r, loc=<0.1974 3.4111 -0.3811>, rot=<0 0 0 1>}
 thumb_01_r {parent=hand_r, loc=<-1.3118 0.3191 2.459>, rot=<0.6546 -0.4042 0.0664 -0.6354>}
 thumb_02_r {parent=thumb_01_r, loc=<-5.9336 0.0448 -0.0401>, rot=<-0.0037 0.0213 0.298 -0.9543>}
 thumb_03_r {parent=thumb_02_r, loc=<-2.2223 -0.3682 0.0907>, rot=<0.0118 -0.013 0.0066 -0.9998>}
 index_metacarpal_r {parent=hand_r, loc=<-1.6535 -0.0183 2.96>, rot=<0.0158 -0.0124 -0.0173 -0.9996>}
 index_01_r {parent=index_metacarpal_r, loc=<-8.3008 -0.0357 0.0524>, rot=<-0.0044 0.0485 0.0924 -0.9945>}
 index_02_r {parent=index_01_r, loc=<-3.3587 0.3004 -0.0452>, rot=<-0.002 0.0027 0.3225 -0.9466>}
 index_03_r {parent=index_02_r, loc=<-2.5578 0.123 0.0252>, rot=<-0.002 0.0067 0.2558 -0.9667>}
 middle_metacarpal_r {parent=hand_r, loc=<-2.1856 0.1725 0.5984>, rot=<0.006 0.0168 -0.0329 -0.9993>}
 middle_01_r {parent=middle_metacarpal_r, loc=<-7.3275 0.2447 -0.0129>, rot=<-0.0208 0.0042 0.0858 -0.9961>}
 middle_02_r {parent=middle_01_r, loc=<-4.0123 -0.1401 -0.1175>, rot=<-0.0218 0.0258 0.3512 -0.9357>}
 middle_03_r {parent=middle_02_r, loc=<-2.6077 -0.0534 -0.1373>, rot=<0.0082 0.0188 0.1708 -0.9851>}
 ring_metacarpal_r {parent=hand_r, loc=<-2.2074 0.3625 -1.3638>, rot=<-0.0284 0.0722 -0.0458 -0.9959>}
 ring_01_r {parent=ring_metacarpal_r, loc=<-7.0295 0.6197 0.1446>, rot=<-0.0489 -0.0252 0.1892 -0.9804>}
 ring_02_r {parent=ring_01_r, loc=<-3.6174 -0.4647 0.1027>, rot=<0 0 0.2806 -0.9598>}
 ring_03_r {parent=ring_02_r, loc=<-2.4152 -0.114 -0.0199>, rot=<0 0 0.1189 -0.9929>}
 pinky_metacarpal_r {parent=hand_r, loc=<-1.7616 0.9753 -3.3814>, rot=<-0.0397 0.0814 -0.069 -0.9935>}
 pinky_01_r {parent=pinky_metacarpal_r, loc=<-7.1709 0.4831 0.1094>, rot=<-0.0654 -0.0035 0.2242 -0.9723>}
 pinky_02_r {parent=pinky_01_r, loc=<-2.8274 -0.0893 0.1043>, rot=<0.0037 0.0011 0.2854 -0.9584>}
 pinky_03_r {parent=pinky_02_r, loc=<-2.2046 -0.0307 0.0078>, rot=<-0.0002 0.0098 0.1445 -0.9895>}
 hand_attach_r {parent=hand_r, loc=<-8.5001 2.9999 0>, rot=<0 0 -0.7071 -0.7071>}
 lowerarm_twist_02_r {parent=lowerarm_r, loc=<-9.6199 0 0>, rot=<0 0 0 -1>}
 lowerarm_twist_01_r {parent=lowerarm_r, loc=<-17.8549 0 0>, rot=<0 0 0 -1>}
 deform_elbow_in_base_r {parent=lowerarm_r, loc=<-0.2111 -5.2591 -0.8275>, rot=<0 0 0 -1>}
 deform_elbow_in_r {parent=deform_elbow_in_base_r, loc=<-0.0001 0 0.0002>, rot=<0 0 0 1>}
 deform_elbow_out_base_r {parent=lowerarm_r, loc=<0 4 -0.0001>, rot=<0 0 0 -1>}
 deform_elbow_out_r {parent=deform_elbow_out_base_r, loc=<0 0 0>, rot=<0 0 0 1>}
 upperarm_twist_01_r {parent=upperarm_r, loc=<-0.0041 0 0>, rot=<0 0 0 -1>}
 deform_upperarm_bk_base_r {parent=upperarm_twist_01_r, loc=<-1.0814 7.2192 1.9455>, rot=<0 0 0 1>}
 deform_upperarm_bk_r {parent=deform_upperarm_bk_base_r, loc=<0 0 0>, rot=<0 0 0 1>}
 deform_upperarm_fr_base_r {parent=upperarm_twist_01_r, loc=<0.3689 -7.0506 2.7652>, rot=<0 0 0 1>}
 deform_upperarm_fr_r {parent=deform_upperarm_fr_base_r, loc=<0 0 0>, rot=<0 0 0 1>}
 deform_upperarm_up_base_r {parent=upperarm_twist_01_r, loc=<3.2214 1.1391 -3.3308>, rot=<0 0 0 1>}
 deform_upperarm_up_r {parent=deform_upperarm_up_base_r, loc=<0 0 0>, rot=<0 0 0 -1>}
 upperarm_twist_02_r {parent=upperarm_r, loc=<-13.7941 0 0>, rot=<0 0 0 -1>}
 deform_pec_aim_base_l {parent=spine_04, loc=<7.8732 12.6388 -2.4123>, rot=<0.0102 -0.4602 0.3979 -0.7936>}
 deform_pec_aim_l {parent=deform_pec_aim_base_l, loc=<0 0 0>, rot=<0 0 0 1>}
 deform_pec_base_l {parent=deform_pec_aim_l, loc=<10.766 0 0>, rot=<0 0 0 1>}
 deform_pec_l {parent=deform_pec_base_l, loc=<0.055 0 0>, rot=<0 0 0 1>}
 deform_pec_aim_base_r {parent=spine_04, loc=<7.8727 12.6388 2.4123>, rot=<-0.4602 -0.0102 0.7936 0.3979>}
 deform_pec_aim_r {parent=deform_pec_aim_base_r, loc=<0 0 0>, rot=<0 0 0 1>}
 deform_pec_base_r {parent=deform_pec_aim_r, loc=<-10.7662 -0.0006 -0.0007>, rot=<0 0 0 1>}
 deform_pec_r {parent=deform_pec_base_r, loc=<0 0 0>, rot=<0 0 0 1>}
 thigh_l {parent=pelvis, loc=<-2.5693 -0.0109 -7.7501>, rot=<0.0012 0.0572 0.0155 -0.9982>}
 calf_l {parent=thigh_l, loc=<-39.0484 0 0>, rot=<0.0006 -0.0161 -0.0391 -0.9991>}
 foot_l {parent=calf_l, loc=<-42.561 0.1077 0.4292>, rot=<0.0418 -0.0455 0.0343 -0.9975>}
 ball_l {parent=foot_l, loc=<-6.1599 13.8264 0.6806>, rot=<0.0001 -0.0001 -0.7063 -0.7079>}
 calf_twist_02_l {parent=calf_l, loc=<-18.8064 -0.0967 0.0397>, rot=<0.0237 0.0095 -0.0035 -0.9997>}
 calf_twist_01_l {parent=calf_l, loc=<-35.813 -0.0967 0.0397>, rot=<0.0237 0.0095 -0.0035 -0.9997>}
 thigh_twist_01_l {parent=thigh_l, loc=<-21.6303 0 0>, rot=<0.0013 0 0 -1>}
 thigh_r {parent=pelvis, loc=<-2.5693 -0.0109 7.7501>, rot=<0.0572 -0.0012 0.9982 0.0155>}
 calf_r {parent=thigh_r, loc=<39.0484 0 0>, rot=<0.0006 -0.0161 -0.0391 -0.9991>}
 foot_r {parent=calf_r, loc=<42.561 -0.1077 -0.4292>, rot=<0.0418 -0.0455 0.0343 -0.9975>}
 ball_r {parent=foot_r, loc=<6.1599 -13.8264 -0.6806>, rot=<0.0001 -0.0001 -0.7063 -0.7079>}
 calf_twist_02_r {parent=calf_r, loc=<18.8064 0.0967 -0.0397>, rot=<0.0237 0.0095 -0.0035 -0.9997>}
 calf_twist_01_r {parent=calf_r, loc=<35.813 0.0967 -0.0397>, rot=<0.0237 0.0095 -0.0035 -0.9997>}
 thigh_twist_01_r {parent=thigh_r, loc=<21.6303 0 0>, rot=<0.0013 0 0 -1>}
 dyn_dress_back_base {parent=pelvis, loc=<-7.8385 -11.6526 0>, rot=<0 0 1 -0.0044>}
 dyn_dress_back_01_mid {parent=dyn_dress_back_base, loc=<5.8473 -0.4683 0.2816>, rot=<0 0 -0.0262 -0.9997>}
 dyn_dress_back_02_mid {parent=dyn_dress_back_01_mid, loc=<22.6133 0.1865 -0.2816>, rot=<0 0 -0.0617 -0.9981>}
 dyn_dress_back_03_mid {parent=dyn_dress_back_02_mid, loc=<20.685 -0.2133 0>, rot=<0 0 -0.0262 -0.9997>}
 dyn_dress_back_04_mid {parent=dyn_dress_back_03_mid, loc=<12.9769 0 0>, rot=<0 0 0 1>}
 dyn_dress_back_base_l {parent=dyn_dress_back_base, loc=<6.347 -1.2183 -8.8684>, rot=<0.0011 -0.0262 -0.0436 -0.9987>}
 dyn_dress_back_01_l {parent=dyn_dress_back_base_l, loc=<0 0 0>, rot=<0 0 0 1>}
 dyn_dress_back_02_l {parent=dyn_dress_back_01_l, loc=<22.065 -0.3876 -0.0617>, rot=<0.0016 0.0081 0 -1>}
 dyn_dress_back_03_l {parent=dyn_dress_back_02_l, loc=<20.6854 0.255 0>, rot=<0 0 -0.0197 -0.9998>}
 dyn_dress_back_04_l {parent=dyn_dress_back_03_l, loc=<12 0 0>, rot=<0 0 0 1>}
 dyn_dress_back_base_r {parent=dyn_dress_back_base, loc=<6.3473 -1.2183 9.4312>, rot=<-0.0011 0.0262 -0.0436 -0.9987>}
 dyn_dress_back_01_r {parent=dyn_dress_back_base_r, loc=<0 0 0>, rot=<0 0 0 1>}
 dyn_dress_back_02_r {parent=dyn_dress_back_01_r, loc=<22.0355 -0.3876 -0.6241>, rot=<0.0016 0.0081 0 -1>}
 dyn_dress_back_03_r {parent=dyn_dress_back_02_r, loc=<20.685 0.255 0>, rot=<0 0 -0.0197 -0.9998>}
 dyn_dress_back_04_r {parent=dyn_dress_back_03_r, loc=<12 0 0>, rot=<0 0 0 1>}
 dyn_dress_front_master_l {parent=pelvis, loc=<-2.5693 -0.0109 -7.7501>, rot=<0.0012 0.0572 0.0155 -0.9982>}
 dyn_dress_front_01_l {parent=dyn_dress_front_master_l, loc=<-24.8122 7.0702 3.0938>, rot=<-0.0128 -0.9994 0.0007 0.0337>}
 dyn_dress_front_02_l {parent=dyn_dress_front_01_l, loc=<17.19 0.0149 0>, rot=<0.0005 0 -0.0058 -1>}
 dyn_dress_front_03_l {parent=dyn_dress_front_02_l, loc=<25.0051 0 0>, rot=<-0.0005 0 0.011 -0.9999>}
 dyn_dress_front_master_r {parent=pelvis, loc=<-2.5693 -0.0109 7.7501>, rot=<0.0572 -0.0012 0.9982 0.0155>}
 dyn_dress_front_01_r {parent=dyn_dress_front_master_r, loc=<24.8122 -7.0701 -3.0938>, rot=<0.9993 -0.0133 -0.0337 0.0002>}
 dyn_dress_front_02_r {parent=dyn_dress_front_01_r, loc=<17.19 0 0>, rot=<0 0 -0.0054 -1>}
 dyn_dress_front_03_r {parent=dyn_dress_front_02_r, loc=<25.0052 0 0>, rot=<0 0.0005 0.9999 0.011>}
 dyn_tassle_master {parent=pelvis, loc=<5.9834 8.7244 -0.2514>, rot=<-0.0127 -0.707 0.0127 -0.707>}
 dyn_tassle_b_01 {parent=dyn_tassle_master, loc=<0.4417 -0.1855 -1.0646>, rot=<-0.0223 -0.6507 -0.0162 -0.7589>}
 dyn_tassle_b_02 {parent=dyn_tassle_b_01, loc=<4.5701 0.288 0>, rot=<0 0 0.0427 -0.9991>}
 dyn_tassle_b_03 {parent=dyn_tassle_b_02, loc=<4.6729 0.0103 0>, rot=<0 0 0.0029 -1>}
 dyn_tassle_a_01 {parent=dyn_tassle_master, loc=<-0.2864 -0.1807 -1.093>, rot=<-0.0204 -0.7296 -0.0185 -0.6833>}
 dyn_tassle_a_02 {parent=dyn_tassle_a_01, loc=<7.9361 0 0>, rot=<0 0 0.0427 -0.9991>}
 dyn_tassle_a_03 {parent=dyn_tassle_a_02, loc=<5.6808 0.0044 0>, rot=<0 0 0.0029 -1>}
 attach {parent=root, loc=<0 0 0>, rot=<0 0 0 1>}
 ik_hand_root {parent=root, loc=<0 0 0>, rot=<0 0 0 1>}
 ik_hand_gun {parent=ik_hand_root, loc=<-51.9727 4.0883 102.6935>, rot=<0.6084 0.1793 0.4613 -0.6205>}
 ik_hand_r {parent=ik_hand_gun, loc=<0 0 0>, rot=<0 0 0 1>}
 ik_hand_l {parent=ik_hand_gun, loc=<53.0285 82.1781 35.2051>, rot=<-0.5102 -0.7906 -0.3387 0>}
 ik_foot_root {parent=root, loc=<0 0 0>, rot=<0 0 0 1>}
 ik_foot_l {parent=ik_foot_root, loc=<15.2608 -4.4176 9.4432>, rot=<0.0357 0.7045 0.0284 -0.7082>}
 ik_foot_r {parent=ik_foot_root, loc=<-15.2608 -4.4176 9.4432>, rot=<0.7082 0.0284 -0.7045 0.0357>}
 attach_fp {parent=root, loc=<0 0 0>, rot=<0 0 0 1>}
 camera_root {parent=root, loc=<0 0 0>, rot=<0 0 0 1>}
 camera {parent=camera_root, loc=<0 0 0>, rot=<0 0 0 1>}
 warp_point_01 {parent=root, loc=<0 0 0>, rot=<0 0 0 1>}
 warp_point_02 {parent=root, loc=<0 0 0>, rot=<0 0 0 1>}
 warp_point_03 {parent=root, loc=<0 0 0>, rot=<0 0 0 1>}
 Pose_1 {type=Curve, loc=<0 0 0>, rot=<0 0 0 1>}
 AbilityStart {type=Curve, loc=<0 0 0>, rot=<0 0 0 1>}
 AbilityEnd {type=Curve, loc=<0 0 0>, rot=<0 0 0 1>}
]
