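# Physics asset for the wings of the male medium LazarusLens Wings NPC body, providing capsule bodies and joints for wing motion.
# Six mirrored chains (a–f) of three to five capsules hang from per-side base bones under wings_base on spine_05, with f sharing e's base.
# Root, spine and base bones carry tiny capsules with free angular joints.
# Each chain's first joint is angularly locked; later joints have tight few-degree limits, narrower on e and f.
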PhysicsAsset /BRCosmetics/Characters/Player/Male/Medium/Bodies/M_MED_LazarusLensWings_NPC/Meshes/M_MED_LazarusLens_Wings_NPC_Physics
bodies {
 dyn_wing_spread_a_01_l [{shapeType=Capsule, length=10.5, radius=6.5, shapeName=dyn_wing_spread_a_01_l_capsule, rotation=<rot 90 0 0>}]
 dyn_wing_spread_a_01_r [{shapeType=Capsule, length=10.5, radius=6.5, shapeName=dyn_wing_spread_a_01_r_capsule, rotation=<rot 90 0 0>}]
 dyn_wing_spread_a_02_l [{shapeType=Capsule, length=10.5, radius=5, shapeName=dyn_wing_spread_a_02_l_capsule, center=<-5 0 0>, rotation=<rot 90 0 0>}]
 dyn_wing_spread_a_02_r [{shapeType=Capsule, length=10.5, radius=5, shapeName=dyn_wing_spread_a_02_r_capsule, center=<5 0 0>, rotation=<rot 90 0 0>}]
 dyn_wing_spread_a_03_l [{shapeType=Capsule, length=6.9, radius=8.6, shapeName=dyn_wing_spread_a_03_l_capsule, center=<-5 0 0>, rotation=<rot 90 0 0>}]
 dyn_wing_spread_a_03_r [{shapeType=Capsule, length=2.8, radius=12.7, shapeName=dyn_wing_spread_a_03_r_capsule, center=<5 0 0>, rotation=<rot 90 0 0>}]
 dyn_wing_spread_b_01_l [{shapeType=Capsule, length=10.5, radius=6.5, shapeName=dyn_wing_spread_b_01_l_capsule, rotation=<rot 90 0 0>}]
 dyn_wing_spread_b_01_r [{shapeType=Capsule, length=10.5, radius=6.5, shapeName=dyn_wing_spread_b_01_r_capsule, rotation=<rot 90 0 0>}]
 dyn_wing_spread_b_02_l [{shapeType=Capsule, length=6.1, radius=10.4, shapeName=dyn_wing_spread_a_01_l_capsule, rotation=<rot 90 0 0>}]
 dyn_wing_spread_b_02_r [{shapeType=Capsule, length=6.1, radius=10.4, shapeName=dyn_wing_spread_a_01_r_capsule, center=<5 0 0>, rotation=<rot 90 0 0>}]
 dyn_wing_spread_b_03_l [{shapeType=Capsule, length=10.5, radius=6, shapeName=dyn_wing_spread_a_02_l_capsule, center=<-5 0 0>, rotation=<rot 90 0 0>}]
 dyn_wing_spread_b_03_r [{shapeType=Capsule, length=10.5, radius=6, shapeName=dyn_wing_spread_a_02_r_capsule, center=<5 0 0>, rotation=<rot 90 0 0>}]
 dyn_wing_spread_b_04_l [{shapeType=Capsule, length=10.5, radius=6, shapeName=dyn_wing_spread_a_03_l_capsule, center=<-5 0 0>, rotation=<rot 90 0 0>}]
 dyn_wing_spread_b_04_r [{shapeType=Capsule, length=10.5, radius=6, shapeName=dyn_wing_spread_a_03_r_capsule, center=<5 0 0>, rotation=<rot 90 0 0>}]
 dyn_wing_spread_b_05_l [{shapeType=Capsule, length=6.7, radius=10, shapeName=dyn_wing_spread_a_04_l_capsule, center=<-5 0 0>, rotation=<rot 90 0 0>}]
 dyn_wing_spread_b_05_r [{shapeType=Capsule, length=6.9, radius=9.6, shapeName=dyn_wing_spread_a_03_r_capsule, center=<5 0 0>, rotation=<rot 90 0 0>}]
 dyn_wing_spread_c_01_l [{shapeType=Capsule, length=10.5, radius=6.5, shapeName=dyn_wing_spread_c_01_l_capsule, rotation=<rot 90 0 0>}]
 dyn_wing_spread_c_01_r [{shapeType=Capsule, length=10.5, radius=6.5, shapeName=dyn_wing_spread_c_01_r_capsule, rotation=<rot 90 0 0>}]
 dyn_wing_spread_c_02_l [{shapeType=Capsule, length=6.1, radius=9.4, shapeName=dyn_wing_spread_c_02_l_capsule, center=<-5 0 0>, rotation=<rot 90 0 0>}]
 dyn_wing_spread_c_02_r [{shapeType=Capsule, length=6.1, radius=9.4, shapeName=dyn_wing_spread_c_02_r_capsule, center=<5 0 0>, rotation=<rot 90 0 0>}]
 dyn_wing_spread_c_03_l [{shapeType=Capsule, length=10.5, radius=5, shapeName=dyn_wing_spread_c_03_l_capsule, center=<-5 0 0>, rotation=<rot 90 0 0>}]
 dyn_wing_spread_c_03_r [{shapeType=Capsule, length=7, radius=8.5, shapeName=dyn_wing_spread_c_03_r_capsule, center=<5 0 0>, rotation=<rot 90 0 0>}]
 dyn_wing_spread_c_04_l [{shapeType=Capsule, length=5.3, radius=10.2, shapeName=dyn_wing_spread_c_04_l_capsule, center=<-5 0 0>, rotation=<rot 90 0 0>}]
 dyn_wing_spread_c_04_r [{shapeType=Capsule, length=2, radius=13.5, shapeName=dyn_wing_spread_c_04_r_capsule, center=<5 0 0>, rotation=<rot 90 0 0>}]
 dyn_wing_spread_d_01_l [{shapeType=Capsule, length=10.5, radius=6.5, shapeName=dyn_wing_spread_d_01_l_capsule, rotation=<rot 90 0 0>}]
 dyn_wing_spread_d_01_r [{shapeType=Capsule, length=10.5, radius=6.5, shapeName=dyn_wing_spread_d_01_r_capsule, rotation=<rot 90 0 0>}]
 dyn_wing_spread_d_02_l [{shapeType=Capsule, length=6.1, radius=9.4, shapeName=dyn_wing_spread_d_02_l_capsule, center=<-5 0 0>, rotation=<rot 90 0 0>}]
 dyn_wing_spread_d_02_r [{shapeType=Capsule, length=6.1, radius=9.4, shapeName=dyn_wing_spread_d_02_r_capsule, center=<5 0 0>, rotation=<rot 90 0 0>}]
 dyn_wing_spread_d_03_l [{shapeType=Capsule, length=10.5, radius=5, shapeName=dyn_wing_spread_d_03_l_capsule, center=<-5 0 0>, rotation=<rot 90 0 0>}]
 dyn_wing_spread_d_03_r [{shapeType=Capsule, length=10.5, radius=5, shapeName=dyn_wing_spread_d_03_r_capsule, center=<5 0 0>, rotation=<rot 90 0 0>}]
 dyn_wing_spread_d_04_l [{shapeType=Capsule, length=1.1, radius=14.4, shapeName=dyn_wing_spread_d_04_l_capsule, center=<-5 0 0>, rotation=<rot 90 0 0>}]
 dyn_wing_spread_d_04_r [{shapeType=Capsule, length=6.9, radius=8.6, shapeName=dyn_wing_spread_d_04_r_capsule, center=<5 0 0>, rotation=<rot 90 0 0>}]
 dyn_wing_spread_e_01_l [{shapeType=Capsule, length=10.5, radius=6.5, shapeName=dyn_wing_spread_e_01_l_capsule, rotation=<rot 90 0 0>}]
 dyn_wing_spread_e_01_r [{shapeType=Capsule, length=10.5, radius=6.5, shapeName=dyn_wing_spread_e_01_r_capsule, rotation=<rot 90 0 0>}]
 dyn_wing_spread_e_02_l [{shapeType=Capsule, length=6.1, radius=9.4, shapeName=dyn_wing_spread_e_02_l_capsule, center=<-5 0 0>, rotation=<rot 90 0 0>}]
 dyn_wing_spread_e_02_r [{shapeType=Capsule, length=6.1, radius=9.4, shapeName=dyn_wing_spread_e_02_r_capsule, center=<5 0 0>, rotation=<rot 90 0 0>}]
 dyn_wing_spread_e_03_l [{shapeType=Capsule, length=10.5, radius=5, shapeName=dyn_wing_spread_e_03_l_capsule, center=<-5 0 0>, rotation=<rot 90 0 0>}]
 dyn_wing_spread_e_03_r [{shapeType=Capsule, length=10.5, radius=5, shapeName=dyn_wing_spread_e_03_r_capsule, center=<5 0 0>, rotation=<rot 90 0 0>}]
 dyn_wing_spread_e_04_l [{shapeType=Capsule, length=2.8, radius=12.7, shapeName=dyn_wing_spread_e_04_l_capsule, center=<-5 0 0>, rotation=<rot 90 0 0>}]
 dyn_wing_spread_e_04_r [{shapeType=Capsule, length=1.9, radius=13.6, shapeName=dyn_wing_spread_e_04_r_capsule, center=<5 0 0>, rotation=<rot 90 0 0>}]
 dyn_wing_spread_f_01_l [{shapeType=Capsule, length=10.5, radius=6.5, shapeName=dyn_wing_spread_e_01_l_capsule, rotation=<rot 90 0 0>}]
 dyn_wing_spread_f_01_r [{shapeType=Capsule, length=10.5, radius=6.5, shapeName=dyn_wing_spread_e_02_r_capsule, rotation=<rot 90 0 0>}]
 dyn_wing_spread_f_02_l [{shapeType=Capsule, length=6.1, radius=9.4, shapeName=dyn_wing_spread_e_02_l_capsule, center=<-5 0 0>, rotation=<rot 90 0 0>}]
 dyn_wing_spread_f_02_r [{shapeType=Capsule, length=6.1, radius=9.4, shapeName=dyn_wing_spread_e_03_r_capsule, center=<5 0 0>, rotation=<rot 90 0 0>}]
 dyn_wing_spread_f_03_l [{shapeType=Capsule, length=10.5, radius=5, shapeName=dyn_wing_spread_e_03_l_capsule, center=<-5 0 0>, rotation=<rot 90 0 0>}]
 dyn_wing_spread_f_03_r [{shapeType=Capsule, length=10.5, radius=5, shapeName=dyn_wing_spread_e_04_r_capsule, center=<5 0 0>, rotation=<rot 90 0 0>}]
 dyn_wing_spread_f_04_l [{shapeType=Capsule, length=10.5, radius=5, shapeName=dyn_wing_spread_e_04_l_capsule, center=<-5 0 0>, rotation=<rot 90 0 0>}]
 dyn_wing_spread_f_04_r [{shapeType=Capsule, length=10.5, radius=5, shapeName=dyn_wing_spread_e_05_r_capsule, center=<5 0 0>, rotation=<rot 90 0 0>}]
 pelvis [{shapeType=Capsule, length=1, radius=1, shapeName=pelvis_capsule}]
 root [{shapeType=Capsule, length=1, radius=1, shapeName=root_capsule}]
 spine_01 [{shapeType=Capsule, length=1, radius=1, shapeName=spine_01_capsule}]
 spine_02 [{shapeType=Capsule, length=1, radius=1, shapeName=spine_02_capsule}]
 spine_03 [{shapeType=Capsule, length=1, radius=1, shapeName=spine_03_capsule}]
 spine_04 [{shapeType=Capsule, length=1, radius=1, shapeName=spine_04_capsule}]
 spine_05 [{shapeType=Capsule, length=1, radius=1, shapeName=spine_05_capsule}]
 wing_spread_a_base_l [{shapeType=Capsule, length=0.505, radius=0.505, shapeName=wing_spread_a_base_l_capsule, rotation=<rot 0 -90 0>}]
 wing_spread_a_base_r [{shapeType=Capsule, length=0.505, radius=0.505, shapeName=wing_spread_a_base_r_capsule, rotation=<rot 0 -90 0>}]
 wing_spread_b_base_l [{shapeType=Capsule, length=0.505, radius=0.505, shapeName=wing_spread_b_base_l_capsule, rotation=<rot 0 -90 0>}]
 wing_spread_b_base_r [{shapeType=Capsule, length=0.505, radius=0.505, shapeName=wing_spread_b_base_r_capsule, rotation=<rot 0 -90 0>}]
 wing_spread_c_base_l [{shapeType=Capsule, length=0.505, radius=0.505, shapeName=wing_spread_c_base_l_capsule, center=<-2 0 0>}]
 wing_spread_c_base_r [{shapeType=Capsule, length=0.505, radius=0.505, shapeName=wing_spread_c_base_r_capsule, rotation=<rot 0 -90 0>}]
 wing_spread_d_base_l [{shapeType=Capsule, length=0.505, radius=0.505, shapeName=wing_spread_d_base_l_capsule, rotation=<rot 0 -90 0>}]
 wing_spread_d_base_r [{shapeType=Capsule, length=0.505, radius=0.505, shapeName=wing_spread_d_base_r_capsule, rotation=<rot 0 -90 0>}]
 wing_spread_e_base_l [{shapeType=Capsule, length=0.505, radius=0.505, shapeName=wing_spread_e_base_l_capsule, rotation=<rot 0 -90 0>}]
 wing_spread_e_base_r [{shapeType=Capsule, length=0.505, radius=0.505, shapeName=wing_spread_e_base_r_capsule, rotation=<rot 0 -90 0>}]
 wings_base [{shapeType=Capsule, length=0.505, radius=0.505, shapeName=wings_base_capsule, rotation=<rot 0 -90 0>}]
}
constraints [
 {
  bone1Name=wing_spread_b_base_l
  bone2Name=wings_base
  swing1LimitDegrees=45
  swing1Motion=Free
  swing2LimitDegrees=45
  swing2Motion=Free
  twistLimitDegrees=45
  twistMotion=Free
 }
 {
  bone1Name=dyn_wing_spread_b_01_l
  bone2Name=wing_spread_b_base_l
  swing1LimitDegrees=3
  swing1Motion=Locked
  swing2LimitDegrees=3
  swing2Motion=Locked
  twistLimitDegrees=3
  twistMotion=Locked
 }
 {
  bone1Name=wing_spread_c_base_l
  bone2Name=wings_base
  swing1LimitDegrees=45
  swing1Motion=Free
  swing2LimitDegrees=45
  swing2Motion=Free
  twistLimitDegrees=45
  twistMotion=Free
 }
 {
  bone1Name=dyn_wing_spread_c_01_l
  bone2Name=wing_spread_c_base_l
  swing1LimitDegrees=3
  swing1Motion=Locked
  swing2LimitDegrees=3
  swing2Motion=Locked
  twistLimitDegrees=3
  twistMotion=Locked
 }
 {bone1Name=dyn_wing_spread_c_02_l, bone2Name=dyn_wing_spread_c_01_l, swing1LimitDegrees=5, swing2LimitDegrees=2.5, twistLimitDegrees=2}
 {bone1Name=dyn_wing_spread_c_03_l, bone2Name=dyn_wing_spread_c_02_l, swing1LimitDegrees=5, swing2LimitDegrees=2.5, twistLimitDegrees=2}
 {bone1Name=dyn_wing_spread_c_04_l, bone2Name=dyn_wing_spread_c_03_l, swing1LimitDegrees=5, swing2LimitDegrees=2.5, twistLimitDegrees=2}
 {
  bone1Name=wing_spread_d_base_l
  bone2Name=wings_base
  swing1LimitDegrees=45
  swing1Motion=Free
  swing2LimitDegrees=45
  swing2Motion=Free
  twistLimitDegrees=45
  twistMotion=Free
 }
 {
  bone1Name=dyn_wing_spread_d_01_l
  bone2Name=wing_spread_d_base_l
  swing1LimitDegrees=3
  swing1Motion=Locked
  swing2LimitDegrees=3
  swing2Motion=Locked
  twistLimitDegrees=3
  twistMotion=Locked
 }
 {bone1Name=dyn_wing_spread_d_02_l, bone2Name=dyn_wing_spread_d_01_l, swing1LimitDegrees=5, swing2LimitDegrees=2.5, twistLimitDegrees=2}
 {bone1Name=dyn_wing_spread_d_03_l, bone2Name=dyn_wing_spread_d_02_l, swing1LimitDegrees=5, swing2LimitDegrees=2.5, twistLimitDegrees=2}
 {bone1Name=dyn_wing_spread_d_04_l, bone2Name=dyn_wing_spread_d_03_l, swing1LimitDegrees=5, swing2LimitDegrees=2.5, twistLimitDegrees=2}
 {
  bone1Name=wing_spread_e_base_l
  bone2Name=wings_base
  swing1LimitDegrees=45
  swing1Motion=Free
  swing2LimitDegrees=45
  swing2Motion=Free
  twistLimitDegrees=45
  twistMotion=Free
 }
 {
  bone1Name=dyn_wing_spread_e_01_l
  bone2Name=wing_spread_e_base_l
  swing1LimitDegrees=3
  swing1Motion=Locked
  swing2LimitDegrees=3
  swing2Motion=Locked
  twistLimitDegrees=3
  twistMotion=Locked
 }
 {bone1Name=dyn_wing_spread_e_02_l, bone2Name=dyn_wing_spread_e_01_l, swing1LimitDegrees=2, swing2LimitDegrees=2.5, twistLimitDegrees=2}
 {bone1Name=dyn_wing_spread_e_03_l, bone2Name=dyn_wing_spread_e_02_l, swing1LimitDegrees=2, swing2LimitDegrees=2.5, twistLimitDegrees=2}
 {bone1Name=dyn_wing_spread_e_04_l, bone2Name=dyn_wing_spread_e_03_l, swing1LimitDegrees=2, swing2LimitDegrees=2.5, twistLimitDegrees=2}
 {
  bone1Name=wing_spread_a_base_l
  bone2Name=wings_base
  swing1LimitDegrees=45
  swing1Motion=Free
  swing2LimitDegrees=45
  swing2Motion=Free
  twistLimitDegrees=45
  twistMotion=Free
 }
 {
  bone1Name=dyn_wing_spread_a_01_l
  bone2Name=wing_spread_a_base_l
  swing1LimitDegrees=3
  swing1Motion=Locked
  swing2LimitDegrees=3
  swing2Motion=Locked
  twistLimitDegrees=3
  twistMotion=Locked
 }
 {bone1Name=dyn_wing_spread_a_02_l, bone2Name=dyn_wing_spread_a_01_l, swing1LimitDegrees=5, swing2LimitDegrees=2.5, twistLimitDegrees=2}
 {bone1Name=dyn_wing_spread_a_03_l, bone2Name=dyn_wing_spread_a_02_l, swing1LimitDegrees=5, swing2LimitDegrees=2.5, twistLimitDegrees=2}
 {
  bone1Name=wing_spread_b_base_r
  bone2Name=wings_base
  swing1LimitDegrees=45
  swing1Motion=Free
  swing2LimitDegrees=45
  swing2Motion=Free
  twistLimitDegrees=45
  twistMotion=Free
 }
 {
  bone1Name=dyn_wing_spread_b_01_r
  bone2Name=wing_spread_b_base_r
  swing1LimitDegrees=3
  swing1Motion=Locked
  swing2LimitDegrees=3
  swing2Motion=Locked
  twistLimitDegrees=3
  twistMotion=Locked
 }
 {
  bone1Name=wing_spread_c_base_r
  bone2Name=wings_base
  swing1LimitDegrees=45
  swing1Motion=Free
  swing2LimitDegrees=45
  swing2Motion=Free
  twistLimitDegrees=45
  twistMotion=Free
 }
 {
  bone1Name=dyn_wing_spread_c_01_r
  bone2Name=wing_spread_c_base_r
  swing1LimitDegrees=3
  swing1Motion=Locked
  swing2LimitDegrees=3
  swing2Motion=Locked
  twistLimitDegrees=3
  twistMotion=Locked
 }
 {bone1Name=dyn_wing_spread_c_02_r, bone2Name=dyn_wing_spread_c_01_r, swing1LimitDegrees=5, swing2LimitDegrees=2.5, twistLimitDegrees=2}
 {bone1Name=dyn_wing_spread_c_03_r, bone2Name=dyn_wing_spread_c_02_r, swing1LimitDegrees=5, swing2LimitDegrees=2.5, twistLimitDegrees=2}
 {bone1Name=dyn_wing_spread_c_04_r, bone2Name=dyn_wing_spread_c_03_r, swing1LimitDegrees=5, swing2LimitDegrees=2.5, twistLimitDegrees=2}
 {
  bone1Name=wing_spread_d_base_r
  bone2Name=wings_base
  swing1LimitDegrees=45
  swing1Motion=Free
  swing2LimitDegrees=45
  swing2Motion=Free
  twistLimitDegrees=45
  twistMotion=Free
 }
 {
  bone1Name=dyn_wing_spread_d_01_r
  bone2Name=wing_spread_d_base_r
  swing1LimitDegrees=3
  swing1Motion=Locked
  swing2LimitDegrees=3
  swing2Motion=Locked
  twistLimitDegrees=3
  twistMotion=Locked
 }
 {bone1Name=dyn_wing_spread_d_02_r, bone2Name=dyn_wing_spread_d_01_r, swing1LimitDegrees=5, swing2LimitDegrees=2.5, twistLimitDegrees=2}
 {bone1Name=dyn_wing_spread_d_03_r, bone2Name=dyn_wing_spread_d_02_r, swing1LimitDegrees=5, swing2LimitDegrees=2.5, twistLimitDegrees=2}
 {bone1Name=dyn_wing_spread_d_04_r, bone2Name=dyn_wing_spread_d_03_r, swing1LimitDegrees=5, swing2LimitDegrees=2.5, twistLimitDegrees=2}
 {
  bone1Name=wing_spread_e_base_r
  bone2Name=wings_base
  swing1LimitDegrees=45
  swing1Motion=Free
  swing2LimitDegrees=45
  swing2Motion=Free
  twistLimitDegrees=45
  twistMotion=Free
 }
 {
  bone1Name=dyn_wing_spread_e_01_r
  bone2Name=wing_spread_e_base_r
  swing1LimitDegrees=3
  swing1Motion=Locked
  swing2LimitDegrees=3
  swing2Motion=Locked
  twistLimitDegrees=3
  twistMotion=Locked
 }
 {bone1Name=dyn_wing_spread_e_02_r, bone2Name=dyn_wing_spread_e_01_r, swing1LimitDegrees=2, swing2LimitDegrees=2.5, twistLimitDegrees=2}
 {bone1Name=dyn_wing_spread_e_03_r, bone2Name=dyn_wing_spread_e_02_r, swing1LimitDegrees=2, swing2LimitDegrees=2.5, twistLimitDegrees=2}
 {bone1Name=dyn_wing_spread_e_04_r, bone2Name=dyn_wing_spread_e_03_r, swing1LimitDegrees=2, swing2LimitDegrees=2.5, twistLimitDegrees=2}
 {
  bone1Name=wing_spread_a_base_r
  bone2Name=wings_base
  swing1LimitDegrees=45
  swing1Motion=Free
  swing2LimitDegrees=45
  swing2Motion=Free
  twistLimitDegrees=45
  twistMotion=Free
 }
 {
  bone1Name=dyn_wing_spread_a_01_r
  bone2Name=wing_spread_a_base_r
  swing1LimitDegrees=3
  swing1Motion=Locked
  swing2LimitDegrees=3
  swing2Motion=Locked
  twistLimitDegrees=3
  twistMotion=Locked
 }
 {bone1Name=dyn_wing_spread_a_02_r, bone2Name=dyn_wing_spread_a_01_r, swing1LimitDegrees=5, swing2LimitDegrees=2.5, twistLimitDegrees=2}
 {bone1Name=dyn_wing_spread_a_03_r, bone2Name=dyn_wing_spread_a_02_r, swing1LimitDegrees=5, swing2LimitDegrees=2.5, twistLimitDegrees=2}
 {
  bone1Name=dyn_wing_spread_f_01_l
  bone2Name=wing_spread_e_base_l
  swing1LimitDegrees=3
  swing1Motion=Locked
  swing2LimitDegrees=3
  swing2Motion=Locked
  twistLimitDegrees=3
  twistMotion=Locked
 }
 {bone1Name=dyn_wing_spread_f_02_l, bone2Name=dyn_wing_spread_f_01_l, swing1LimitDegrees=2, swing2LimitDegrees=2.5, twistLimitDegrees=2}
 {bone1Name=dyn_wing_spread_f_03_l, bone2Name=dyn_wing_spread_f_02_l, swing1LimitDegrees=2, swing2LimitDegrees=2.5, twistLimitDegrees=2}
 {bone1Name=dyn_wing_spread_f_04_l, bone2Name=dyn_wing_spread_f_03_l, swing1LimitDegrees=2, swing2LimitDegrees=2.5, twistLimitDegrees=2}
 {
  bone1Name=dyn_wing_spread_f_01_r
  bone2Name=wing_spread_e_base_r
  swing1LimitDegrees=3
  swing1Motion=Locked
  swing2LimitDegrees=3
  swing2Motion=Locked
  twistLimitDegrees=3
  twistMotion=Locked
 }
 {bone1Name=dyn_wing_spread_f_02_r, bone2Name=dyn_wing_spread_f_01_r, swing1LimitDegrees=2, swing2LimitDegrees=2.5, twistLimitDegrees=2}
 {bone1Name=dyn_wing_spread_f_03_r, bone2Name=dyn_wing_spread_f_02_r, swing1LimitDegrees=2, swing2LimitDegrees=2.5, twistLimitDegrees=2}
 {bone1Name=dyn_wing_spread_f_04_r, bone2Name=dyn_wing_spread_f_03_r, swing1LimitDegrees=2, swing2LimitDegrees=2.5, twistLimitDegrees=2}
 {
  bone1Name=wings_base
  bone2Name=spine_05
  swing1LimitDegrees=45
  swing1Motion=Free
  swing2LimitDegrees=45
  swing2Motion=Free
  twistLimitDegrees=45
  twistMotion=Free
 }
 {
  bone1Name=pelvis
  bone2Name=root
  swing1LimitDegrees=45
  swing1Motion=Free
  swing2LimitDegrees=45
  swing2Motion=Free
  twistLimitDegrees=45
  twistMotion=Free
 }
 {
  bone1Name=spine_01
  bone2Name=pelvis
  swing1LimitDegrees=45
  swing1Motion=Free
  swing2LimitDegrees=45
  swing2Motion=Free
  twistLimitDegrees=45
  twistMotion=Free
 }
 {
  bone1Name=spine_02
  bone2Name=spine_01
  swing1LimitDegrees=45
  swing1Motion=Free
  swing2LimitDegrees=45
  swing2Motion=Free
  twistLimitDegrees=45
  twistMotion=Free
 }
 {
  bone1Name=spine_04
  bone2Name=spine_03
  swing1LimitDegrees=45
  swing1Motion=Free
  swing2LimitDegrees=45
  swing2Motion=Free
  twistLimitDegrees=45
  twistMotion=Free
 }
 {
  bone1Name=spine_03
  bone2Name=spine_02
  swing1LimitDegrees=45
  swing1Motion=Free
  swing2LimitDegrees=45
  swing2Motion=Free
  twistLimitDegrees=45
  twistMotion=Free
 }
 {
  bone1Name=spine_05
  bone2Name=spine_04
  swing1LimitDegrees=45
  swing1Motion=Free
  swing2LimitDegrees=45
  swing2Motion=Free
  twistLimitDegrees=45
  twistMotion=Free
 }
 {bone1Name=dyn_wing_spread_b_02_r, bone2Name=dyn_wing_spread_b_01_r, swing1LimitDegrees=5, swing2LimitDegrees=2.5, twistLimitDegrees=2}
 {bone1Name=dyn_wing_spread_b_03_r, bone2Name=dyn_wing_spread_b_02_r, swing1LimitDegrees=5, swing2LimitDegrees=2.5, twistLimitDegrees=2}
 {bone1Name=dyn_wing_spread_b_04_r, bone2Name=dyn_wing_spread_b_03_r, swing1LimitDegrees=5, swing2LimitDegrees=2.5, twistLimitDegrees=2}
 {bone1Name=dyn_wing_spread_b_05_r, bone2Name=dyn_wing_spread_b_04_r, swing1LimitDegrees=5, swing2LimitDegrees=2.5, twistLimitDegrees=2}
 {bone1Name=dyn_wing_spread_b_02_l, bone2Name=dyn_wing_spread_b_01_l, swing1LimitDegrees=5, swing2LimitDegrees=2.5, twistLimitDegrees=2}
 {bone1Name=dyn_wing_spread_b_03_l, bone2Name=dyn_wing_spread_b_02_l, swing1LimitDegrees=5, swing2LimitDegrees=2.5, twistLimitDegrees=2}
 {bone1Name=dyn_wing_spread_b_04_l, bone2Name=dyn_wing_spread_b_03_l, swing1LimitDegrees=5, swing2LimitDegrees=2.5, twistLimitDegrees=2}
 {bone1Name=dyn_wing_spread_b_05_l, bone2Name=dyn_wing_spread_b_04_l, swing1LimitDegrees=5, swing2LimitDegrees=2.5, twistLimitDegrees=2}
]
