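# Anim Blueprint for the Lazarus Lens winged NPC body, animating wings, skirt and bun over the parent mesh's copied pose.
# A state machine driven by speed, crouch and LifeDrain ability curves layers wing animations from wings_base.
# Dress constraints and a Control Rig feed RigidBody wing and crouch-switched skirt physics, AnimDynamics bun, then corrective bone offsets.
# The EventGraph sets wing physics weight from jumping/falling and back-dress lift from knee bend.
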AnimBlueprint /BRCosmetics/Characters/Player/Male/Medium/Bodies/M_MED_LazarusLensWings_NPC/Meshes/M_MED_LazarusLens_Wings_NPC_AnimBP
target_skeleton=/Game/Characters/Player/Male/Medium/Bodies/M_MED_LazarusLensWings_NPC/Meshes/M_MED_LazarusLens_Wings_NPC_Skeleton
graphs [
 AnimGraph {nodes [
   "Transform (Modify) Bone - Bone: wing_spread_b_base_r"(Alpha=1, Translation="0,0,0", Rotation="-5,0,-5", Scale="1,1,1") {
    rotationMode=BMM_Additive
    rotationSpace=BCS_BoneSpace
    alphaBoolBlend {blendOption=Linear}
    boneToModify {boneName=wing_spread_b_base_r}
    rotation {pitch=-5, roll=-5}
   }
   "Transform (Modify) Bone - Bone: wing_spread_b_base_l"(Alpha=1, Translation="0,0,0", Rotation="5,0,-5", Scale="1,1,1") {
    rotationMode=BMM_Additive
    rotationSpace=BCS_BoneSpace
    alphaBoolBlend {blendOption=Linear}
    boneToModify {boneName=wing_spread_b_base_l}
    rotation {pitch=5, roll=-5}
   }
   n1 = RigidBody(ComponentPose=Local To Component(LocalPose=Use cached pose 'wings_pose'()), Alpha=Get wing_dynamic_alpha() -> (wing_dynamic_alpha:double)) {
    bOverrideWorldGravity=true
    overridePhysicsAsset=&/M_MED_LazarusLens_Wings_NPC_Physics
    simulationSpace=BaseBoneSpace
    alphaBoolBlend {blendOption=Linear}
    alphaScaleBiasClamp {bInterpResult=true}
    baseBoneRef {boneName=pelvis}
    componentAppliedLinearAccClamp=<2000 2000 2500>
    componentLinearAccScale=<0.05 0.05 0.1>
    componentLinearVelScale=<0.05 0.05 0.1>
    overrideWorldGravity {z=-980}
   }
   n2 = Get bIsCrouching()
   n3 = RigidBody!(ComponentPose=@n1, bAlphaBoolEnabled:bool=@n2, ExternalForce) -> (Pose) {
    alphaInputType=Bool
    bOverrideWorldGravity=true
    overridePhysicsAsset=&/M_MED_LazarusLens_Skirt_Crouch_NPC_Physics
    simulationSpace=BaseBoneSpace
    alphaBoolBlend {blendOption=Linear}
    baseBoneRef {boneName=pelvis}
    componentAppliedLinearAccClamp=<2000 2000 2000>
    componentLinearAccScale=<0.2 0.2 0.3>
    componentLinearVelScale=<0.2 0.2 0.3>
    overrideWorldGravity {z=-980}
   }
   n4 = RigidBody(ComponentPose=@n3, bAlphaBoolEnabled=NOT Boolean(A=@n2)) {
    alphaInputType=Bool
    bOverrideWorldGravity=true
    overridePhysicsAsset=&/M_MED_LazarusLens_Skirt_NPC_Physics
    simulationSpace=BaseBoneSpace
    alphaBoolBlend {blendOption=Linear}
    baseBoneRef {boneName=pelvis}
    componentAppliedLinearAccClamp=<2000 2000 2000>
    componentLinearAccScale=<0.2 0.2 0.3>
    componentLinearVelScale=<0.2 0.2 0.3>
    overrideWorldGravity {z=-980}
   }
   n5 = Anim Dynamics - Bone: dyn_bun_02(ComponentPose=@n4, Alpha=0.75, ExternalForce="0,0,0") {
    angularDampingOverride=0.9
    angularSpringConstant=20000
    bOverrideAngularDamping=true
    bUseGravityOverride=true
    simulationSpace=RootRelative
    alphaBoolBlend {blendOption=Linear}
    boundBone {boneName=dyn_bun_02}
    componentAppliedLinearAccClamp=<2000 2000 3000>
    componentLinearAccScale=<0.15 0.15 0.2>
    componentLinearVelScale=<0.15 0.15 0.2>
    gravityOverride {z=-980}
    physicsBodyDefinitions [{
      sphereCollisionRadius=0
      boundBone {boneName=dyn_bun_02}
      boxExtents=<15 12 12>
      constraintSetup {angularLimitsMax=<10 10 10>, angularLimitsMin=<-5 -10 -3>, angularTarget=<1 0 0>}
      localJointOffset=<7 0 0>
     }]
    relativeSpaceBone {boneName=head}
   }
   n6 = Anim Dynamics - Bone: dyn_bun_03(ComponentPose=@n5, Alpha=0.75, ExternalForce="0,0,0") {
    angularDampingOverride=0.9
    angularSpringConstant=20000
    bOverrideAngularDamping=true
    bUseGravityOverride=true
    simulationSpace=RootRelative
    alphaBoolBlend {blendOption=Linear}
    boundBone {boneName=dyn_bun_03}
    componentAppliedLinearAccClamp=<1500 1500 2500>
    componentLinearAccScale=<0.2 0.2 0.3>
    componentLinearVelScale=<0.2 0.2 0.3>
    gravityOverride {z=-980}
    physicsBodyDefinitions [{
      sphereCollisionRadius=0
      boundBone {boneName=dyn_bun_03}
      boxExtents=<6 9 9>
      constraintSetup {angularLimitsMax=<15 15 15>, angularLimitsMin=<-15 -15 -15>, angularTarget=<1 0 0>}
      localJointOffset=<3 0 0>
     }]
    relativeSpaceBone {boneName=head}
   }
   n7 = "Transform (Modify) Bone - Bone: wing_spread_c_base_r"(ComponentPose=@n6, Alpha=1, Translation="0,0,0", Rotation="0,-5,0", Scale="1,1,1") {
    rotationMode=BMM_Additive
    rotationSpace=BCS_BoneSpace
    alphaBoolBlend {blendOption=Linear}
    boneToModify {boneName=wing_spread_c_base_r}
    rotation {yaw=-5}
   }
   n8 = "Transform (Modify) Bone - Bone: wing_spread_d_base_r"(ComponentPose=@n7, Alpha=1, Translation="0,0,0", Rotation="-0.000,-10.000,-0.000", Scale="1,1,1") {
    rotationMode=BMM_Additive
    rotationSpace=BCS_BoneSpace
    alphaBoolBlend {blendOption=Linear}
    boneToModify {boneName=wing_spread_d_base_r}
    rotation=<rot 0 -10 0>
   }
   n9 = "Transform (Modify) Bone - Bone: wing_spread_e_base_r"(ComponentPose=@n8, Alpha=1, Translation="0,0,0", Rotation="0,-25,0", Scale="1,1,1") {
    rotationMode=BMM_Additive
    rotationSpace=BCS_BoneSpace
    alphaBoolBlend {blendOption=Linear}
    boneToModify {boneName=wing_spread_e_base_r}
    rotation {yaw=-25}
   }
   n10 = "Transform (Modify) Bone - Bone: dyn_wing_spread_e_03_r"(ComponentPose=@n9, Alpha=1, Translation="0,0,0", Rotation="0,10,0", Scale="1,1,1") {
    rotationMode=BMM_Additive
    rotationSpace=BCS_BoneSpace
    alphaBoolBlend {blendOption=Linear}
    boneToModify {boneName=dyn_wing_spread_e_03_r}
    rotation {yaw=10}
   }
   n11 = "Transform (Modify) Bone - Bone: dyn_wing_spread_e_04_r"(ComponentPose=@n10, Alpha=1, Translation="0,0,0", Rotation="0,10,0", Scale="1,1,1") {
    rotationMode=BMM_Additive
    rotationSpace=BCS_BoneSpace
    alphaBoolBlend {blendOption=Linear}
    boneToModify {boneName=dyn_wing_spread_e_04_r}
    rotation {yaw=10}
   }
   n12 = "Transform (Modify) Bone - Bone: wing_spread_f_base_r"(ComponentPose=@n11, Alpha=1, Translation="0,0,0", Rotation="0,8,0", Scale="1,1,1") {
    rotationMode=BMM_Additive
    rotationSpace=BCS_BoneSpace
    alphaBoolBlend {blendOption=Linear}
    boneToModify {boneName=wing_spread_f_base_r}
    rotation {yaw=8}
   }
   n13 = "Transform (Modify) Bone - Bone: dyn_wing_spread_c_03_l"(ComponentPose=@n12, Alpha=1, Translation="0,0,0", Rotation="0,-5,0", Scale="1,1,1") {
    rotationMode=BMM_Additive
    rotationSpace=BCS_BoneSpace
    alphaBoolBlend {blendOption=Linear}
    boneToModify {boneName=dyn_wing_spread_c_03_l}
    rotation {yaw=-5}
   }
   n14 = "Transform (Modify) Bone - Bone: wing_spread_d_base_l"(ComponentPose=@n13, Alpha=1, Translation="0,0,0", Rotation="-0.000,-10.000,-0.000", Scale="1,1,1") {
    rotationMode=BMM_Additive
    rotationSpace=BCS_BoneSpace
    alphaBoolBlend {blendOption=Linear}
    boneToModify {boneName=wing_spread_d_base_l}
    rotation=<rot 0 -10 0>
   }
   n15 = "Transform (Modify) Bone - Bone: wing_spread_e_base_l"(ComponentPose=@n14, Alpha=1, Translation="0,0,0", Rotation="0,-25,0", Scale="1,1,1") {
    rotationMode=BMM_Additive
    rotationSpace=BCS_BoneSpace
    alphaBoolBlend {blendOption=Linear}
    boneToModify {boneName=wing_spread_e_base_l}
    rotation {yaw=-25}
   }
   n16 = "Transform (Modify) Bone - Bone: dyn_wing_spread_e_03_l"(ComponentPose=@n15, Alpha=1, Translation="0,0,0", Rotation="0,10,0", Scale="1,1,1") {
    rotationMode=BMM_Additive
    rotationSpace=BCS_BoneSpace
    alphaBoolBlend {blendOption=Linear}
    boneToModify {boneName=dyn_wing_spread_e_03_l}
    rotation {yaw=10}
   }
   n17 = "Transform (Modify) Bone - Bone: dyn_wing_spread_e_04_l"(ComponentPose=@n16, Alpha=1, Translation="0,0,0", Rotation="0,10,0", Scale="1,1,1") {
    rotationMode=BMM_Additive
    rotationSpace=BCS_BoneSpace
    alphaBoolBlend {blendOption=Linear}
    boneToModify {boneName=dyn_wing_spread_e_04_l}
    rotation {yaw=10}
   }
   n18 = "Transform (Modify) Bone - Bone: wing_spread_f_base_l"(ComponentPose=@n17, Alpha=1, Translation="0,0,0", Rotation="0,8,0", Scale="1,1,1") {
    rotationMode=BMM_Additive
    rotationSpace=BCS_BoneSpace
    alphaBoolBlend {blendOption=Linear}
    boneToModify {boneName=wing_spread_f_base_l}
    rotation {yaw=8}
   }
   n19 = "Constraint - wing_base_r (PP)"!(ComponentPose=@n18, Alpha=1, ConstraintWeights_0=0.9, ConstraintWeights_1:float=0.1) -> (Pose) {
    alphaBoolBlend {blendOption=Linear}
    boneToModify {boneName=wing_base_r}
    constraintSetup [{targetBone {boneName=spine_05}}, {targetBone {boneName=clavicle_r}}]
    constraintWeights [1,1]
   }
   n20 = "Constraint - wing_base_l (PP)"!(ComponentPose=@n19, Alpha=1, ConstraintWeights_0=0.9, ConstraintWeights_1:float=0.1) -> (Pose) {
    alphaBoolBlend {blendOption=Linear}
    boneToModify {boneName=wing_base_l}
    constraintSetup [{targetBone {boneName=spine_05}}, {targetBone {boneName=clavicle_l}}]
    constraintWeights [1,1]
   }
   n21 = "Transform (Modify) Bone - Bone: wing_base_r"!(ComponentPose=@n20, bAlphaBoolEnabled:bool=True, Translation="1,0,3", Rotation="10,-5,5", Scale="1,1,1") -> (Pose) {
    alphaInputType=Bool
    rotationMode=BMM_Additive
    rotationSpace=BCS_BoneSpace
    translationMode=BMM_Additive
    translationSpace=BCS_BoneSpace
    alphaBoolBlend {blendInTime=0.5, blendOption=Linear, blendOutTime=0.5}
    boneToModify {boneName=wing_base_r}
    rotation=<rot 10 -5 5>
    translation {x=1, z=3}
   }
   n22 = "Transform (Modify) Bone - Bone: wing_base_l"!(ComponentPose=@n21, bAlphaBoolEnabled:bool=True, Translation="-1,0,-3", Rotation="10,-5,5", Scale="1,1,1") -> (Pose) {
    alphaInputType=Bool
    rotationMode=BMM_Additive
    rotationSpace=BCS_BoneSpace
    translationMode=BMM_Additive
    translationSpace=BCS_BoneSpace
    alphaBoolBlend {blendInTime=0.5, blendOption=Linear, blendOutTime=0.5}
    boneToModify {boneName=wing_base_l}
    rotation=<rot 10 -5 5>
    translation {x=-1, z=-3}
   }
   n23 = "Transform (Modify) Bone - Bone: dyn_dress_back_base"(ComponentPose=@n22, Alpha=1, Translation="0,0,0", Rotation="0,5,0", Scale="1,1,1") {
    rotationMode=BMM_Additive
    rotationSpace=BCS_BoneSpace
    translationMode=BMM_Additive
    translationSpace=BCS_BoneSpace
    alphaBoolBlend {blendOption=Linear}
    boneToModify {boneName=dyn_dress_back_base}
    rotation {yaw=5}
   }
   n24 = Get calf_l_alpha() -> (calf_l_alpha:double)
   n25 = "Transform (Modify) Bone - Bone: dyn_dress_back_01_mid"(ComponentPose=@n23, Alpha=@n24, Translation="0,0,0", Rotation="0,25,0", Scale="1,1,1") {
    rotationMode=BMM_Additive
    rotationSpace=BCS_BoneSpace
    alphaBoolBlend {blendOption=Linear}
    boneToModify {boneName=dyn_dress_back_01_mid}
    rotation {yaw=25}
   }
   n26 = Get calf_l_alpha()
   n27 = "Transform (Modify) Bone - Bone: dyn_dress_back_01_l"(ComponentPose=@n25, Alpha=@n26, Translation="0,0,0", Rotation="0,25,0", Scale="1,1,1") {
    rotationMode=BMM_Additive
    rotationSpace=BCS_BoneSpace
    alphaBoolBlend {blendOption=Linear}
    boneToModify {boneName=dyn_dress_back_01_l}
    rotation {yaw=25}
   }
   n28 = Get calf_r_alpha() -> (calf_r_alpha:double)
   n29 = "Transform (Modify) Bone - Bone: dyn_dress_back_01_r"(ComponentPose=@n27, Alpha=@n28, Translation="0,0,0", Rotation="0,25,0", Scale="1,1,1") {
    rotationMode=BMM_Additive
    rotationSpace=BCS_BoneSpace
    alphaBoolBlend {blendOption=Linear}
    boneToModify {boneName=dyn_dress_back_01_r}
    rotation {yaw=25}
   }
   n30 = "Transform (Modify) Bone - Bone: dyn_dress_back_02_mid"(ComponentPose=@n29, Alpha=@n24, Translation="0,0,0", Rotation="0,15,0", Scale="1,1,1") {
    rotationMode=BMM_Additive
    rotationSpace=BCS_BoneSpace
    alphaBoolBlend {blendOption=Linear}
    boneToModify {boneName=dyn_dress_back_02_mid}
    rotation {yaw=15}
   }
   n31 = "Transform (Modify) Bone - Bone: dyn_dress_back_02_l"(ComponentPose=@n30, Alpha=@n26, Translation="0,0,0", Rotation="0,15,0", Scale="1,1,1") {
    rotationMode=BMM_Additive
    rotationSpace=BCS_BoneSpace
    alphaBoolBlend {blendOption=Linear}
    boneToModify {boneName=dyn_dress_back_02_l}
    rotation {yaw=15}
   }
   n32 = "Transform (Modify) Bone - Bone: dyn_dress_back_02_r"(ComponentPose=@n31, Alpha=@n28, Translation="0,0,0", Rotation="0,15,0", Scale="1,1,1") {
    rotationMode=BMM_Additive
    rotationSpace=BCS_BoneSpace
    alphaBoolBlend {blendOption=Linear}
    boneToModify {boneName=dyn_dress_back_02_r}
    rotation {yaw=15}
   }
   Output Pose(Result=Component To Local(ComponentPose=@n32)) {layerGroup=None}
   n33 = Copy Pose From Mesh() {bCopyCurves=true, bUseAttachedParent=true}
   AnimGraphNode_StateMachine_0 = wing_position() {is=title AnimGraphNode_StateMachine}
   n34 = "Transform (Modify) Bone - Bone: wings_base"(
    ComponentPose=Local To Component(LocalPose=Layered blend per bone(BasePose=@n33, BlendPoses_0=@AnimGraphNode_StateMachine_0) {layerSetup [{branchFilters [{boneName=wings_base}]}]})
    Alpha=1
    Translation="0,0,0"
    Rotation="0,0,0"
    Scale="1,1,1"
   ) {
    rotationSpace=BCS_BoneSpace
    scaleMode=BMM_Replace
    scaleSpace=BCS_BoneSpace
    alphaBoolBlend {blendOption=Linear}
    boneToModify {boneName=wings_base}
   }
   n35 = "Constraint - dyn_dress_back_base (PPP)"!(ComponentPose=@n34, Alpha=1, ConstraintWeights_0=0.8, ConstraintWeights_1:float=0.1, ConstraintWeights_2:float=0.1) -> (Pose) {
    alphaBoolBlend {blendOption=Linear}
    boneToModify {boneName=dyn_dress_back_base}
    constraintSetup [{targetBone {boneName=pelvis}}, {targetBone {boneName=thigh_twist_01_l}}, {targetBone {boneName=thigh_twist_01_r}}]
    constraintWeights [1,1,1]
   }
   n36 = "Constraint - dyn_dress_front_master_l (R)"(ComponentPose=@n35, Alpha=1, ConstraintWeights_0=1) {
    alphaBoolBlend {blendOption=Linear}
    boneToModify {boneName=dyn_dress_front_master_l}
    constraintSetup [{transformType=Rotation, targetBone {boneName=thigh_l}}]
    constraintWeights [1]
   }
   n37 = "Constraint - dyn_dress_front_master_r (R)"(ComponentPose=@n36, Alpha=1, ConstraintWeights_0=1) {
    alphaBoolBlend {blendOption=Linear}
    boneToModify {boneName=dyn_dress_front_master_r}
    constraintSetup [{transformType=Rotation, targetBone {boneName=thigh_r}}]
    constraintWeights [1]
   }
   n38 = Control Rig(Source=Component To Local(ComponentPose=@n37)) {
    alphaBoolBlend {blendOption=Linear}
    controlRigAssetReference {blueprintRigClass=&/M_MED_LazarusLens_Wings_NPC_CtrlRig.M_MED_LazarusLens_Wings_NPC_CtrlRig_C}
   }
   Save cached pose 'wings_pose'(Pose=@n38) {cacheName=wings_pose}
  ]}
 wing_position {
  entry=Base
  path=AnimGraph.AnimGraphNode_StateMachine_0.wing_position
  ids=id
  nodes [
   AnimStateNode_0 = Crouch() {is=title AnimStateNode}
   AnimStateNode_1 = Base() {is=title AnimStateNode}
   AnimStateNode_2 = LifeDrain_Start() {is=title AnimStateNode}
   AnimStateNode_3 = LifeDrain_Loop() {is=title AnimStateNode}
   AnimStateNode_4 = LifeDrain_End() {is=title AnimStateNode}
   AnimStateNode_5 = Walk() {is=title AnimStateNode}
   AnimStateNode_6 = Run() {is=title AnimStateNode}
   AnimStateTransitionNode_0 = AnimStateTransitionNode()
   AnimStateTransitionNode_1 = AnimStateTransitionNode()
   AnimStateTransitionNode_2 = AnimStateTransitionNode()
   AnimStateTransitionNode_3 = AnimStateTransitionNode() {bAutomaticRuleBasedOnSequencePlayerInState=true, blendMode=Linear, crossfadeDuration=0}
   AnimStateTransitionNode_4 = AnimStateTransitionNode()
   AnimStateTransitionNode_5 = AnimStateTransitionNode() {blendMode=Linear, crossfadeDuration=0.25}
   AnimStateTransitionNode_6 = AnimStateTransitionNode() {crossfadeDuration=0.25}
   AnimStateTransitionNode_7 = AnimStateTransitionNode()
   AnimStateTransitionNode_8 = AnimStateTransitionNode() {blendMode=Linear, crossfadeDuration=0.25}
   AnimStateTransitionNode_9 = AnimStateTransitionNode() {bAutomaticRuleBasedOnSequencePlayerInState=true, blendMode=Linear, crossfadeDuration=0.3}
  ]
  links [
   {
    dst=Crouch
    rule=AnimGraph.AnimGraphNode_StateMachine_0.wing_position.AnimStateTransitionNode_0.Transition
    src=Base
    via=AnimStateTransitionNode_0
   }
   {
    dst=LifeDrain_Start
    rule=AnimGraph.AnimGraphNode_StateMachine_0.wing_position.AnimStateTransitionNode_8.Transition
    src=Base
    via=AnimStateTransitionNode_8
   }
   {
    dst=Walk
    rule=AnimGraph.AnimGraphNode_StateMachine_0.wing_position.AnimStateTransitionNode_1.Transition
    src=Base
    via=AnimStateTransitionNode_1
   }
   {
    dst=Base
    rule=AnimGraph.AnimGraphNode_StateMachine_0.wing_position.AnimStateTransitionNode_2.Transition
    src=Crouch
    via=AnimStateTransitionNode_2
   }
   {
    dst=Base
    rule=AnimGraph.AnimGraphNode_StateMachine_0.wing_position.AnimStateTransitionNode_9.Transition
    src=LifeDrain_End
    via=AnimStateTransitionNode_9
   }
   {
    dst=LifeDrain_End
    rule=AnimGraph.AnimGraphNode_StateMachine_0.wing_position.AnimStateTransitionNode_5.AnimationTransitionGraph_0
    src=LifeDrain_Loop
    via=AnimStateTransitionNode_5
   }
   {
    dst=LifeDrain_Loop
    rule=AnimGraph.AnimGraphNode_StateMachine_0.wing_position.AnimStateTransitionNode_3.AnimationTransitionGraph_0
    src=LifeDrain_Start
    via=AnimStateTransitionNode_3
   }
   {
    dst=Walk
    rule=AnimGraph.AnimGraphNode_StateMachine_0.wing_position.AnimStateTransitionNode_7.Transition
    src=Run
    via=AnimStateTransitionNode_7
   }
   {
    dst=Base
    rule=AnimGraph.AnimGraphNode_StateMachine_0.wing_position.AnimStateTransitionNode_4.Transition
    src=Walk
    via=AnimStateTransitionNode_4
   }
   {
    dst=Run
    rule=AnimGraph.AnimGraphNode_StateMachine_0.wing_position.AnimStateTransitionNode_6.Transition
    src=Walk
    via=AnimStateTransitionNode_6
   }
  ]
 }
 Crouch {
  path=AnimGraph.AnimGraphNode_StateMachine_0.wing_position.AnimStateNode_0.Crouch
  nodes [Output Animation Pose(Result=Sequence Player 'Lazarus_Wings_Closed_Temp'() {sequence=&/Lazarus_Wings_Closed_Temp})]
 }
 Base {
  path=AnimGraph.AnimGraphNode_StateMachine_0.wing_position.AnimStateNode_1.Base
  nodes [Output Animation Pose(Result=Sequence Player 'Lazarus_Wings_Idle'() {sequence=/BRCosmetics/Characters/Player/Male/Medium/Bodies/M_MED_LazarusLensWings_NPC/Anim/Lazarus_Wings_Idle})]
 }
 LifeDrain_Start {
  path=AnimGraph.AnimGraphNode_StateMachine_0.wing_position.AnimStateNode_2.LifeDrain_Start
  nodes [
   n1 = Sequence Player 'LazarusLensWings_LifeDrain_Start'() {
    bLoopAnimation=false
    sequence=/BRCosmetics/Characters/Player/Male/Medium/Bodies/M_MED_LazarusLensWings_NPC/Anim/LazarusLensWings_LifeDrain_Start
   }
   Output Animation Pose(Result=@n1)
  ]
 }
 LifeDrain_Loop {
  path=AnimGraph.AnimGraphNode_StateMachine_0.wing_position.AnimStateNode_3.LifeDrain_Loop
  nodes [Output Animation Pose(Result=Sequence Player 'LazarusLensWings_LifeDrain_Loop'() {sequence=/BRCosmetics/Characters/Player/Male/Medium/Bodies/M_MED_LazarusLensWings_NPC/Anim/LazarusLensWings_LifeDrain_Loop})]
 }
 LifeDrain_End {
  path=AnimGraph.AnimGraphNode_StateMachine_0.wing_position.AnimStateNode_4.LifeDrain_End
  nodes [
   n1 = Sequence Player 'LazarusLensWings_LifeDrain_End'() {
    bLoopAnimation=false
    sequence=/BRCosmetics/Characters/Player/Male/Medium/Bodies/M_MED_LazarusLensWings_NPC/Anim/LazarusLensWings_LifeDrain_End
   }
   Output Animation Pose(Result=@n1)
  ]
 }
 Walk {
  path=AnimGraph.AnimGraphNode_StateMachine_0.wing_position.AnimStateNode_5.Walk
  nodes [Output Animation Pose(Result=Sequence Player 'Lazarus_Wings_Walk'() {sequence=/BRCosmetics/Characters/Player/Male/Medium/Bodies/M_MED_LazarusLensWings_NPC/Anim/Lazarus_Wings_Walk})]
 }
 Run {
  path=AnimGraph.AnimGraphNode_StateMachine_0.wing_position.AnimStateNode_6.Run
  nodes [Output Animation Pose(Result=Sequence Player 'Lazarus_Wings_Run'() {sequence=/BRCosmetics/Characters/Player/Male/Medium/Bodies/M_MED_LazarusLensWings_NPC/Anim/Lazarus_Wings_Run})]
 }
 Transition {
  path=AnimGraph.AnimGraphNode_StateMachine_0.wing_position.AnimStateTransitionNode_0.Transition
  nodes [Result(bCanEnterTransition=Get bIsCrouching())]
 }
 Transition {
  path=AnimGraph.AnimGraphNode_StateMachine_0.wing_position.AnimStateTransitionNode_1.Transition
  nodes [Result(bCanEnterTransition="float > float"(A=Get PawnSpeed2D()))]
 }
 Transition {
  path=AnimGraph.AnimGraphNode_StateMachine_0.wing_position.AnimStateTransitionNode_2.Transition
  nodes [
   Result(bCanEnterTransition=NOT Boolean(A=Get bIsCrouching()))
   Get bIsTacticalSprinting()
  ]
 }
 AnimationTransitionGraph_0 {path=AnimGraph.AnimGraphNode_StateMachine_0.wing_position.AnimStateTransitionNode_3.AnimationTransitionGraph_0, nodes [Result()]}
 Transition {
  path=AnimGraph.AnimGraphNode_StateMachine_0.wing_position.AnimStateTransitionNode_4.Transition
  nodes [
   n1 = Get PawnSpeed2D()
   Result(bCanEnterTransition=NOT Boolean(A="float > float"(A=@n1)))
  ]
 }
 AnimationTransitionGraph_0 {
  path=AnimGraph.AnimGraphNode_StateMachine_0.wing_position.AnimStateTransitionNode_5.AnimationTransitionGraph_0
  nodes [Result(bCanEnterTransition=Get AbilityEnd() -> (AbilityEnd:bool))]
 }
 Transition {
  path=AnimGraph.AnimGraphNode_StateMachine_0.wing_position.AnimStateTransitionNode_6.Transition
  nodes [Result(bCanEnterTransition="float > float"(A=Get PawnSpeed2D(), B=350))]
 }
 Transition {
  path=AnimGraph.AnimGraphNode_StateMachine_0.wing_position.AnimStateTransitionNode_7.Transition
  nodes [
   n1 = Get PawnSpeed2D()
   Result(bCanEnterTransition=NOT Boolean(A="float > float"(A=@n1, B=350)))
  ]
 }
 Transition {
  path=AnimGraph.AnimGraphNode_StateMachine_0.wing_position.AnimStateTransitionNode_8.Transition
  nodes [Result(bCanEnterTransition=Get AbilityStart() -> (AbilityStart:bool))]
 }
 Transition {
  path=AnimGraph.AnimGraphNode_StateMachine_0.wing_position.AnimStateTransitionNode_9.Transition
  nodes [Result(bCanEnterTransition=Get AbilityEnd())]
 }
 EventGraph {nodes [
   Event BlueprintUpdateAnimation()
   > n1 = Sequence()
   > n2 = Branch(Condition=OR Boolean(A=Get bIsJumping(), B=Get bIsFalling()))
   > Set wing_dynamic_alpha(wing_dynamic_alpha:double=0.75) -> (Output_Get:double)
   Set wing_dynamic_alpha(execute=@n2.else, wing_dynamic_alpha=0.425)
   Set running(execute=@n1.then_1, running:bool="float > float"(A=Get PawnSpeed2D(), B=350)) -> (Output_Get:bool)
   n3 = GetOwnerSkeletalMeshComponent()
   n4 = BreakTransform(InTransform=GetDeltaTransformFromRefPose(self=@n3, BoneName=calf_r, BaseName=thigh_r))
   > Set calf_r_alpha(calf_r_alpha:double=MapRangeClamped(Value=BreakRotator(InRot=@n4.Rotation).Yaw, InRangeA=0, InRangeB=90)) -> (Output_Get:double)
   n5 = BreakTransform(InTransform=GetDeltaTransformFromRefPose(self=@n3, BoneName=calf_l, BaseName=thigh_l))
   > Set calf_l_alpha(calf_l_alpha:double=MapRangeClamped(Value=BreakRotator(InRot=@n5.Rotation).Yaw, InRangeA=0, InRangeB=90)) -> (Output_Get:double)
   n6 = BreakTransform(InTransform=GetDeltaTransformFromRefPose(self=@n3, BoneName=thigh_r, BaseName=pelvis))
   > Set thigh_r_alpha(thigh_r_alpha:double=MapRangeClamped(Value=BreakRotator(InRot=@n6.Rotation).Yaw, InRangeA=0, InRangeB=90)) -> (Output_Get:double)
   n7 = BreakTransform(InTransform=GetDeltaTransformFromRefPose(self=@n3, BoneName=thigh_l, BaseName=pelvis))
   > Set thigh_l_alpha(thigh_l_alpha:double=MapRangeClamped(Value=BreakRotator(InRot=@n7.Rotation).Yaw, InRangeA=0, InRangeB=90)) -> (Output_Get:double)
   n8 = "Equal (Float)"(A=GetCurveValue(CurveName=AbilityStart), B=1)
   > Set AbilityStart(AbilityStart:bool=@n8) -> (Output_Get:bool)
   > Set AbilityEnd(AbilityEnd:bool="Equal (Float)"(A=GetCurveValue(CurveName=AbilityEnd), B=1)) -> (Output_Get:bool)
   > n9 = Branch(Condition=@n8)
   > Set not_doing_ability(not_doing_ability:bool=false) -> (Output_Get:bool)
   Set not_doing_ability(execute=@n9.else, not_doing_ability=true)
  ]}
]
variables [
 AbilityEnd {type=boolean}
 AbilityStart {type=boolean}
 calf_l_alpha {type=number}
 calf_r_alpha {type=number}
 not_doing_ability {type=boolean}
 running {type=boolean}
 thigh_l_alpha {type=number}
 thigh_r_alpha {type=number}
 wing_dynamic_alpha {type=number}
]
